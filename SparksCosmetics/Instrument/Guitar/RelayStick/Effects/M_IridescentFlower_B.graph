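# Masked two-sided lit material for iridescent flower mesh particles in the RelayStick guitar cosmetic's effects.
# Particle color, optionally hue-shifted per face side and randomized, feeds a metal reflectance LUT.
# Thin-film spectrum lookups warped by noise, a view-dependent hue-shifted rainbow and time-of-day rim lighting add color layers, with dynamic parameters for blend controls.
# A dynamic parameter also drives axis rotation around the particle pivot, with corrected normals.
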Material /SparksCosmetics/Instrument/Guitar/RelayStick/Effects/M_IridescentFlower_B
graph {
 properties {
  bUsedWithNiagaraMeshParticles=true
  blendMode=BLEND_Masked
  pixelDepthOffsetMode=PDOM_Legacy
  previewMesh=&/SM_RelayStick_Flower_C
  translucencyLightingMode=TLM_Surface
  twoSided=true
  usedShadingModels=Default Lit
 }
 nodes [
  n1 = TextureCoordinate()
  n2 = Time()
  n3 = LinearInterpolate(Alpha=ParticleRandom()) {constB=10}
  n4 = Frac(None=Panner(Coordinate=Multiply(A=@n1) {constB=0.1}, Time=Add(A=@n2, B=@n3)) {speedY=0.5})
  n5 = OneMinus(None=@n4)
  n6 = Multiply(A=MaterialFunctionCall(UV Channel=Max(A=@n5, B=@n4)) {materialFunction=&LinearGradient}.VGradient)
  Add(A=MaterialFunctionCall() {materialFunction=&BoundingBoxBased_0-1_UVW}.B, B=@n6)
  n7 = MaterialFunctionCall() {materialFunction=&BoundingBoxBased_0-1_UVW}
  Add(A=@n7.B) {constB=0.1}
  n8 = DynamicParameter() {parameterIndex=1, paramNames [RotationAmount, Param2, Param3, Param4]}
  Add(
   A=@n8.RotationAmount
   B=Multiply(A=Time() {period=5}, B=ScalarParameter() {defaultValue=0.1, group=FanRotation, parameterName=FanRotationSpeed})
  )
  CollectionParameter() {collection=/FMAudioReactiveData/AudioAnalysis/MPC_Sparks_Audio_Parameters, parameterName=BeatLifetime}
  Constant() {r=1}
  n9 = NamedRerouteUsage()
  n10 = Desaturation(
   None=LinearInterpolate(A=NamedRerouteUsage(), B=NamedRerouteUsage(), Alpha=Saturate(None=@n9)) {constAlpha=0.1}
   Fraction=ScalarParameter() {defaultValue=0.01, parameterName=Desaturate Fraction}
  )
  LinearInterpolate(A=@n10, B=NamedRerouteUsage(), Alpha=NamedRerouteUsage())
  MaterialFunctionCall!(Alpha, Materials) {materialFunction=/Game/Global/MaterialFunctions/TechArt/MF_MetalLUT_TechArt}
  MaterialFunctionCall(Emissive=Multiply(A=NamedRerouteUsage(), B=NamedRerouteUsage()) {constB=0.25}) {materialFunction=&MF_FX_EmissiveAdjustments}
  n11 = DynamicParameter() {defaultValue=<rgba 1 0.5 0.25 0>, paramNames [2SidedColor On/Off, ThinFilmBlend, Emission Boost, Rainbow Mix]}
  NamedRerouteDeclaration(None=@n11.Emission Boost) {name=EmissionBoost, nodeColor=<rgba 1 0.7734 0 1>}
  n12 = ScalarParameter() {defaultValue=0.1, parameterName=opacity}
  n13 = Fresnel(
   ExponentIn=ScalarParameter() {defaultValue=1, parameterName=Fresnel opacity exp}
   BaseReflectFractionIn=ScalarParameter() {defaultValue=-0.5, parameterName=Base reflection}
  ) {baseReflectFraction=-0.5, exponent=1}
  n14 = LinearInterpolate(Alpha=@n13) {constA=7, constAlpha=2}
  NamedRerouteDeclaration(None=Multiply(A=LinearInterpolate(A=@n12, B=@n13), B=Clamp(None=@n14) {maxDefault=0.9, minDefault=0.6})) {name=Fresnel, nodeColor=<rgba 0 0.1094 1 1>}
  n15 = Constant() {desc=masking, r=1}
  n16 = Saturate(None=Multiply(A=@n15, B=@n15))
  n17 = VertexNormalWS()
  n18 = Multiply(
   A=Fresnel(ExponentIn=ScalarParameter() {defaultValue=2, group=RimLight-Close, parameterName=RimFresnelExp}) {baseReflectFraction=0}
   B=Multiply(
    A=@n16
    B=MaterialFunctionCall(Float3=@n17) {materialFunction=&BreakOutFloat3Components}.B
   )
  )
  n19 = ScalarParameter() {defaultValue=0.03, group=RimLight, parameterName=RimContrast}
  NamedRerouteDeclaration(None=Saturate(None=MaterialFunctionCall(In=@n18, Contrast=@n19) {materialFunction=&CheapContrast}.Result)) {name=RimLightFresnel, nodeColor=<rgba 1 0.375 0 1>}
  n20 = NamedRerouteDeclaration(None=@n10) {name=BaseColorOut, nodeColor=<rgba 0.3125 1 0 1>}
  NamedRerouteDeclaration(None=Multiply(A=@n20, B=Saturate(None=@n7.B))) {name=EmissiveColorPreRim, nodeColor=<rgba 0.3359 0 1 1>}
  n21 = NamedRerouteUsage()
  n22 = NamedRerouteUsage()
  NamedRerouteDeclaration(None=Add(A=NamedRerouteUsage(), B=Multiply(A=@n21, B=@n22))) {name=EmissiveColorWithRim, nodeColor=<rgba 0.8281 1 0 1>}
  n23 = Abs(None=DotProduct(A=CameraVectorWS(), B=VertexNormalWS())) {desc=for 2 sided}
  n24 = Constant() {desc=edge sharpness, r=2}
  n25 = OneMinus(None=Saturate(None=Power(Base=@n23, Exp=@n24)))
  NamedRerouteDeclaration(None=@n25) {name=test, nodeColor=<rgba 0.4531 1 0 1>}
  NamedRerouteDeclaration(None=Frac(None=@n8.RotationAmount)) {name=l_FanRotationAmount, nodeColor=<rgba 0 1 0.2031 1>}
  n26 = VectorParameter() {group=FanRotation, parameterName=NormalizedRotationAxis, defaultValue=<rgba 0 0 1 0>}
  NamedRerouteDeclaration(None=Normalize(VectorInput=Transform(None=@n26.RGB) {transformSourceType=TRANSFORMSOURCE_Instance})) {name=l_FanRotationAxis, nodeColor=<rgba 0.5469 0 1 1>}
  n27 = Multiply(A=TextureCoordinate(), B=ScalarParameter() {defaultValue=1, parameterName=Normal Tiling}) {constB=0.75}
  NamedRerouteDeclaration(None=LinearInterpolate(
    A=Constant3Vector() {constant=<rgba 0 0 1 0>}
    B=TextureSample(UVs=@n27) {samplerType=SAMPLERTYPE_Normal, texture=/Game/Global/Textures/Default/Tileable_Patterns/1Normal}.RGB
   ) {constAlpha=0.175}) {name=in N, nodeColor=<rgba 0 1 0.9766 1>}
  n28 = NamedRerouteUsage()
  NamedRerouteDeclaration(None=MaterialFunctionCall(Rotation Axis=NamedRerouteUsage(), Rotation Angle=NamedRerouteUsage(), World Space Vertex Normals=Transform(None=@n28)) {materialFunction=&FixRotateAboutAxisNormals}.New Tangent Space Vertex Normal) {name=outN, nodeColor=<rgba 0.4062 1 0 1>}
  n29 = MaterialFunctionCall() {materialFunction=&ObjectPivotPoint}
  n30 = VectorParameter() {group=FanRotation, parameterName=PivotPointBias}
  NamedRerouteDeclaration(None=RotateAboutAxis(
    NormalizedRotationAxis=NamedRerouteUsage()
    RotationAngle=NamedRerouteUsage()
    PivotPoint=Add(A=@n29.Mesh Particle Pivot Location, B=@n30.RGB)
    Position=WorldPosition().XYZ
   )) {name=outWPO, nodeColor=<rgba 1 0 0.3516 1>}
  n31 = TextureCoordinate()
  n32 = ScalarParameter() {defaultValue=4, parameterName=UvFilmOffsetTiling}
  n33 = ChannelMaskParameter(None=TextureSampleParameter2D(UVs=Multiply(A=@n31, B=@n32)) {parameterName=ThinFilmUVOffsetTexture, texture=/Game/Global/Textures/Default/Noise/3xPerlin_M}.RGB) {group=06_Thin Film, maskChannel=Blue, parameterName=ThinFilmUVOffsetChannel, defaultValue=<rgba 0 0 1 0>}
  n34 = Multiply(A=@n33.R) {constB=0.1}
  n35 = ScalarParameter() {defaultValue=0.2, group=06_Thin Film, parameterName=ThinFilmUVOffsetIntensity}
  n36 = ScalarParameter() {defaultValue=0.5, group=06_Thin Film, parameterName=ThinFilmWarp, sortPriority=4}
  n37 = StaticSwitchParameter(True=@n25, False=@n6) {parameterName=Use Camera Based Iridescent}
  n38 = Add(
   A=Multiply(A=ScalarParameter() {defaultValue=1, group=06_Thin Film, parameterName=ThinFilmScale, sortPriority=5}, B=Power(Base=@n37, Exp=@n36))
   B=ScalarParameter() {group=06_Thin Film, parameterName=ThinFilmOffset}
  )
  n39 = StaticSwitchParameter(True=Add(A=Multiply(A=@n34, B=@n35), B=@n38), False=@n38) {defaultValue=true, parameterName=UseThinFilmUVOffset}
  n40 = TextureSampleParameter2D(UVs=@n39) {
   group=06_Thin Film
   parameterName=Thin Film Texture
   sortPriority=1
   texture=/Game/Global/Textures/Default/Thinfilm/T_ThinFilm_Spectrum_COLOR
  }
  n41 = NamedRerouteUsage()
  n42 = Add(A=@n41, B=Multiply(A=@n40.RGB, B=ScalarParameter() {defaultValue=2, group=06_Thin Film, parameterName=ThinFilmIntensity, sortPriority=2}))
  n43 = Power(Base=@n37, Exp=ScalarParameter() {defaultValue=2, group=06_Thin Film, parameterName=ThinFilmExponent, sortPriority=3})
  n44 = LinearInterpolate(Alpha=Constant() {r=0.75}) {constA=1, constB=0.01}
  n45 = Saturate(None=LinearInterpolate(A=@n43, B=Multiply(A=@n44, B=@n43), Alpha=ScalarParameter() {defaultValue=0.1, group=06_Thin Film, parameterName=RoughnessInfluence}))
  NamedRerouteDeclaration(None=LinearInterpolate(A=@n41, B=LinearInterpolate(A=@n41, B=@n42, Alpha=@n45), Alpha=NamedRerouteUsage()) {desc=Thin Film Blend}) {name=ThinFilmColor, nodeColor=<rgba 0 1 0.8125 1>}
  NamedRerouteDeclaration(None=Saturate(None=@n11.ThinFilmBlend)) {name=ThinFilmBlend, nodeColor=<rgba 1 0.0234 0 1>}
  n46 = Time()
  n47 = ScalarParameter() {defaultValue=25, parameterName=Rainbow Rotater}
  n48 = DotProduct(A=Rotator(Coordinate=PixelNormalWS(), Time=Multiply(A=@n46, B=@n47)), B=CameraVectorWS())
  n49 = Multiply(A=TextureCoordinate(), B=ScalarParameter() {defaultValue=1, parameterName=Rainbow Uv})
  n50 = MaterialFunctionCall(
   Hue Shift Percentage=Multiply(A=Add(A=@n48, B=@n49), B=ScalarParameter() {defaultValue=1, parameterName=hue shift percentage}) {constB=0.2}
   Texture=VectorParameter() {parameterName=Hue colors, defaultValue=<rgba 1 0.473 0 1>}.RGB
  ) {materialFunction=&HueShift}
  n51 = Power(Base=Fresnel(), Exp=ScalarParameter() {defaultValue=1, parameterName=Fresnel Outer Power})
  NamedRerouteDeclaration(None=Multiply(A=Multiply(A=@n50.Result, B=@n51), B=ScalarParameter() {defaultValue=3, parameterName=Emissive Color})) {name=Rainbow, nodeColor=<rgba 0.0547 1 0 1>}
  NamedRerouteDeclaration(None=@n11.Rainbow Mix) {name=Rainbow lerp, nodeColor=<rgba 0 0.2734 1 1>}
  n52 = Divide(A=VertexInterpolator(VS=PreSkinnedPosition()).PS, B=Constant() {r=170})
  NamedRerouteDeclaration(None=Saturate(None=ComponentMask(None=@n52) {b=true})) {name=Z-Gradient, nodeColor=<rgba 0 1 0.7891 1>}
  n53 = NamedRerouteUsage()
  n54 = VectorParameter() {group=RimLight, parameterName=RimMorning, defaultValue=<rgba 0.875 0.8626 0.7474 1>}
  n55 = VectorParameter() {group=RimLight, parameterName=RimDay, defaultValue=<rgba 1.2 0.9092 0.5938 1>}
  n56 = VectorParameter() {group=RimLight, parameterName=RimEvening, defaultValue=<rgba 0.9635 0.5427 0.2158 1>}
  n57 = VectorParameter() {group=RimLight, parameterName=RimNight, defaultValue=<rgba 0.3927 0.3462 0.651 1>}
  n58 = Desaturation(None=Clamp(None=NamedRerouteUsage()) {maxDefault=0.7, minDefault=0.03}, Fraction=Constant() {r=0.6})
  n59 = MaterialFunctionCall!(
   BaseColor=@n58
   Gmap=@n16
   LightVector=VectorParameter() {group=RimLight, parameterName=RimLightDirection01, defaultValue=<rgba -2 1 2 1>}.RGB
  ) {materialFunction=/Game/Packages/Fortress_SharedMaterials/Base_Material_Functions/MF_Rimlight_SingleLight}
  n60 = MaterialFunctionCall(
   BaseColor=@n58
   Gmap=@n16
   LightVector=VectorParameter() {group=RimLight, parameterName=RimLightDirection02, defaultValue=<rgba 0.6 0.5 1.5 1>}.RGB
  ) {materialFunction=&MF_Rimlight_SingleLight}
  n61 = Multiply(
   A=Multiply(
    A=MaterialFunctionCall(Color_Morning=@n54.RGB, Color_Day=@n55.RGB, Color_Evening=@n56.RGB, Color_Night=@n57.RGB) {materialFunction=&Color_Time_of_Day}.Result
    B=Add(A=@n59.Result, B=@n60.Result)
   )
   B=ScalarParameter() {defaultValue=25, desc="artemis = 2.2 - Asteria 6(pending)", group=RimLight, parameterName=RimLightIntensity}
  )
  NamedRerouteDeclaration(None=Clamp(None=Multiply(A=@n53, B=@n61)) {maxDefault=500}) {name=RimLightColors, nodeColor=<rgba 0 1 0.1094 1>}
  NamedRerouteUsage()
  NamedRerouteUsage()
  NamedRerouteUsage()
  NamedRerouteUsage()
  NamedRerouteUsage()
  ParticleColor()
  n62 = TextureCoordinate()
  n63 = MaterialFunctionCall!(
   Width=Constant() {r=0.6}
   Sharpness
   Offset
   Direction Switch
   Compare
   Input Coordinates=Panner(Coordinate=Multiply(A=@n62) {constB=3}) {speedY=0.3}
  ) {materialFunction=/Engine/Functions/Engine_MaterialFunctions03/Procedurals/GeneratedBand}
  Reroute(None=Multiply(A=Power(Base=@n63.Result) {constExponent=10}, B=ScalarParameter() {defaultValue=0.1, parameterName=movement}))
  n64 = CameraVectorWS()
  n65 = PixelNormalWS()
  n66 = CurveAtlasRowParameter(CurveTime=Clamp(None=DotProduct(A=@n64, B=@n65))) {
   atlas=/Game/Global/Textures/Default/Curve/FN_Character_Atlas
   curve=/Game/Global/Textures/Default/Curve/AluminumReflectance_Weak
   defaultValue=2
   group=06_Metal LUT
   parameterName=Metal LUT Curve
  }
  n67 = ParticleColor()
  n68 = TwoSidedSign()
  n69 = MaterialFunctionCall(Hue Shift Percentage=ScalarParameter() {defaultValue=0.15, parameterName=Hue Shift Percent}, Texture=@n67.RGB) {materialFunction=&HueShift}
  n70 = LinearInterpolate(A=@n67.RGB, B=@n69.Result, Alpha=@n68)
  n71 = LinearInterpolate(A=@n69.Result, B=@n67.RGB, Alpha=@n68)
  n72 = ParticleRandom()
  n73 = NamedRerouteDeclaration(None=LinearInterpolate(A=@n67.RGB, B=LinearInterpolate(A=@n70, B=@n71, Alpha=@n72), Alpha=Saturate(None=@n11.2SidedColor On/Off))) {name=ParticleColor, nodeColor=<rgba 1 0.3984 0 1>}
  n74 = Multiply(A=@n73, B=ChannelMaskParameter(None=@n66) {group=06_Metal LUT, maskChannel=Green, parameterName=LUTChannel, defaultValue=<rgba 0 1 0 0>}.R)
  n75 = Saturate(None=ScalarParameter() {defaultValue=1, group=06_Metal LUT, parameterName=MetalLutIntensity})
  Reroute(None=NamedRerouteDeclaration(None=LinearInterpolate(A=@n73, B=@n74, Alpha=@n75)) {name=MetalColor, nodeColor=<rgba 0.2891 0 1 1>})
  Reroute(None=@n20)
  Reroute(None=Constant()) {desc=spec}
  Reroute()
  ScalarParameter() {defaultValue=0.125, parameterName=Roughness}
  StaticSwitch(
   True=LinearInterpolate(A=Constant() {r=1}, B=ScalarParameter() {defaultValue=0.1, parameterName=IQR}, Alpha=Fresnel())
   Value=StaticBoolParameter() {parameterName=use refraction}
  )
  TextureSampleParameter2D() {parameterName=Mask, texture=&T_WhitesRGB}
  TextureSample(UVs=@n39) {samplerSource=SSM_Wrap_WorldGroupSettings, texture=&T_ThinFilm_Spectrum_COLOR}
  Transform()
  VectorParameter() {parameterName=base color, defaultValue=<rgba 0.2215 0.8396 1 1>}
 ]
}
