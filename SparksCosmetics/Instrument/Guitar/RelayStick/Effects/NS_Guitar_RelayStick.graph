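# Niagara effect for the RelayStick guitar cosmetic in Sparks, emitting colored flower mesh particles with lights around the instrument.
# Burst and continuous-rate emitters pick from three mesh indices and a color palette, offsetting box-shaped spawn positions from skeletal mesh samples.
# Particles drift with curl noise, grow and shrink over life, and pulse in size with the audio BeatLifetime parameter.
# The original Flower_local and InitialBurst emitters are disabled in favor of their 001 variants.
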NiagaraSystem /SparksCosmetics/Instrument/Guitar/RelayStick/Effects/NS_Guitar_RelayStick
dependencies {
 usedDataInterfaces [
  &NiagaraDataInterfaceArrayColor
  &NiagaraDataInterfaceArrayFloat
  &NiagaraDataInterfaceCurve
  &NiagaraDataInterfaceArrayFloat3
  &NiagaraDataInterfaceArrayInt32
  &NiagaraDataInterfaceSkeletalMesh
 ]
 usedDynamicInputs [&SelectFloatFromArray]
 usedModules [
  &self:SystemSpawnScript.SystemScriptSource.SystemScriptGraph.NiagaraNodeAssignment_0.SetVariables_#1
  &SystemState
 ]
}
emitters {
 Flower_local {
  bEnabled=false
  emitterUpdateScript [
   &EmitterState(
    Inactive Response=::NewEnumerator1
    Life Cycle Mode=::NewEnumerator1
    Loop Delay=&RandomRangeFloat(Minimum=0.5, Maximum=1)
    Loop Duration=10
    UseLoopDelay=false
   )
   &SpawnRate(Spawn Probability=0.9, Use Spawn Probability=true)
  ]
  particleSpawnScript [
   &PartitionParticles(Number Of Partitions=3, Particles Per Partition=1, Partition Mode:ENiagara_PartitionMode=::NewEnumerator1)
   /SetVariables(Particles.MeshIndex=$Particles.PartitionParticles.Partition) {isSetParameters=true, module=SetVariables_#2}
   &InitializeParticle(
    Color=&SelectLinearColorFromArray(
     Array Sampling Mode=::NewEnumerator0
     Direct Array Index=?
     Color Selection Array=DataInterface() {propertyValues {ColorData [
        <rgba 0.0275 0.5765 0.9961 1>
        <rgba 0.2968 0.1864 0.5089 1>
        <rgba 0.6939 0.104 0.6346 1>
        <rgba 0.1479 0.6172 0.6172 1>
        <rgba 1 0.324 0 1>
        <rgba 1 0.9258 0 1>
        <rgba 0.7969 0.0913 0.0913 1>
       ]}}
     Random Seed=0
    )
    Lifetime Max=5
    Lifetime Min=4
    Lifetime Mode=::NewEnumerator1
    Mesh Scale Mode=::NewEnumerator1
    Mesh Uniform Scale Max=0.35
    Mesh Uniform Scale Min=0.2
    Sprite Size Mode=::NewEnumerator1
    UsePositionOffset=false
    Write Material Random=true
   )
   &ShapeLocation(
    Box Size=<25 40 1>
    MaskSpawnGroup=false
    Override Seed=false
    Rotation Angle=90
    Rotation Axis=<0 1 0>
    Rotation Mode=::NewEnumerator1
    Shape Primitive=::NewEnumerator2
    Surface Only=false
    UseOwnerScale=true
   )
   &InitialMeshOrientation(Use Rotation Vector=false) {module=InitialMeshOrientation001}
   &DynamicMaterialParameters(
    Index 0 Param 1=1
    Index 0 Param 2=0.25
    Index 0 Param 3=0.5
    Module.Index0Param0Write=true
    Module.Index0Param1Write=true
    Module.Index0Param2Write=true
    Module.Index0Param3Write=true
   )
   &AddVelocity(
    Constrain To Radius=25
    Module.Use Falloff Distance=true
    Origin Offset=<0 -5 0>
    Override Seed=false
    Velocity Mode=::NewEnumerator1
    Velocity Speed=&RandomRangeFloat(Minimum=3, Maximum=6)
   )
  ]
  particleUpdateScript [
   &ParticleState()
   &ScaleMeshSize(Scale Factor=&VectorFromFloat(Value=0.75)) {module=ScaleMeshSize003}
   &SkeletalMeshLocation(
    Bone / Socket Sampling Mode=::NewEnumerator5
    MaskSpawnGroup=false
    Mesh Sampling Type=::NewEnumerator6
    Orientation Sampling=::NewEnumerator1
    Position Sampling=::NewEnumerator1
    Sampled Position Offset=&AddVectorToPosition(
     Position=$Particles.ShapeLocation.ShapePosition
     Vector=&RandomRangeVector(Minimum=<0 1 0>, Maximum=<0 2 0>)
    )
    Sampled Position Offset Coordinate Space:ENiagaraSkelMeshTransforms=::NewEnumerator1
    Skeletal Mesh=$System.Skeletal Mesh
   )
   &CurlNoiseForce(Noise Strength=5, Pan Noise=true, Pan Noise Field=<0 0 4>)
   &ScaleMeshSize(Scale Factor=&VectorFromFloat(Value=&Lerp_Float(
      A=1
      B=1.7
      Alpha=&FloatFromCurve(
       FloatCurve=DataInterface() {propertyValues {Curve {keys [
           {arriveTangent=9.0725, interpMode=RCIM_Cubic, time=0, value=0}
           <at 0.1 1 RCIM_Cubic>
           {arriveTangent=-0.0088, interpMode=RCIM_Cubic, leaveTangent=-0.0088, tangentMode=RCTM_User, time=1, value=0}
          ]}}}
       CurveIndex=$NPC.NPC_Sparks_Audio_Parameters001.BeatLifetime
      )
     ))) {module=ScaleMeshSize001}
   &DynamicMaterialParameters(
    Index 0 Param 1=1
    Index 0 Param 2=0.25
    Index 0 Param 3=0.5
    Index 1 Param 1=&FloatFromCurve(
     FloatCurve=DataInterface() {propertyValues {Curve {keys [<at 0 1>,<at 1 0>]}}}
     Scale Curve=&RandomRangeFloat(Minimum=0.25, Maximum=0.5)
    )
    Module.Index0Param0Write=true
    Module.Index0Param1Write=true
    Module.Index0Param2Write=true
    Module.Index0Param3Write=true
    Module.Index1Param0Write=true
    Module.Index1Param1Write=false
    Module.Index1Param2Write=false
    Module.Index1Param3Write=false
    Write Parameter Index 1=true
   ) {module=DynamicMaterialParameters001}
   &ScaleMeshSize(Scale Factor=&VectorFromFloat(Value=&FloatFromCurve(FloatCurve=DataInterface() {propertyValues {Curve {keys [
          {arriveTangent=0.3651, interpMode=RCIM_Cubic, leaveTangent=0.3651, tangentMode=RCTM_User, time=0, value=0}
          {arriveTangent=-2.5249, leaveTangent=-2.5249, time=0.1408, value=0.9847}
          {arriveTangent=-1.146, interpMode=RCIM_Cubic, leaveTangent=-1.146, time=0.6, value=1}
          {arriveTangent=-0.0405, interpMode=RCIM_Cubic, leaveTangent=-0.0405, tangentMode=RCTM_User, time=1, value=0}
         ]}}})))
   &ScaleMeshSize(Scale Factor=&MakeVector(
     X=&FloatFromCurve(FloatCurve=DataInterface() {propertyValues {Curve {keys [
          {arriveTangent=1.8704, interpMode=RCIM_Cubic, leaveTangent=1.8704, tangentMode=RCTM_User, time=0, value=0}
          {arriveTangent=0.2799, leaveTangent=1.1179, tangentMode=RCTM_User, time=0.3, value=1}
          {arriveTangent=0.0224, interpMode=RCIM_Cubic, leaveTangent=0.0224, tangentMode=RCTM_User, time=1, value=1}
         ]}}})
     Y=&FloatFromCurve(FloatCurve=DataInterface() {propertyValues {Curve {keys [
          {arriveTangent=2.4662, interpMode=RCIM_Cubic, leaveTangent=2.4662, tangentMode=RCTM_User, time=0, value=0}
          {arriveTangent=0.422, leaveTangent=1.1179, tangentMode=RCTM_User, time=0.3, value=1}
          {arriveTangent=0.0224, interpMode=RCIM_Cubic, leaveTangent=0.0224, tangentMode=RCTM_User, time=1, value=1}
         ]}}})
     Z=&FloatFromCurve(FloatCurve=DataInterface() {propertyValues {Curve {keys [
          {interpMode=RCIM_Cubic, tangentMode=RCTM_User, time=0, value=0}
          {arriveTangent=0.3068, leaveTangent=0.3068, time=0.2, value=1.025}
          {arriveTangent=0.0743, interpMode=RCIM_Cubic, leaveTangent=0.0743, tangentMode=RCTM_User, time=1, value=1}
         ]}}})
    )) {enabled=false, module=ScaleMeshSize004}
   &ScaleMeshSize(Scale Factor=&MakeVector(
     X=&FloatFromCurve(FloatCurve=DataInterface() {propertyValues {Curve {keys [
          {tangentMode=RCTM_User, time=0, value=1}
          {arriveTangent=-3.4324, interpMode=RCIM_Cubic, leaveTangent=-3.4324, tangentMode=RCTM_User, time=0.7, value=1}
          {arriveTangent=-0.4311, interpMode=RCIM_Cubic, leaveTangent=-0.4311, tangentMode=RCTM_User, time=1, value=0}
         ]}}})
     Y=&FloatFromCurve(FloatCurve=DataInterface() {propertyValues {Curve {keys [
          {tangentMode=RCTM_User, time=0, value=1}
          {arriveTangent=-3.3527, interpMode=RCIM_Cubic, leaveTangent=-3.3527, tangentMode=RCTM_User, time=0.7, value=1}
          {arriveTangent=-0.4311, interpMode=RCIM_Cubic, leaveTangent=-0.4311, tangentMode=RCTM_User, time=1, value=0}
         ]}}})
     Z=&FloatFromCurve(FloatCurve=DataInterface() {propertyValues {Curve {keys [{interpMode=RCIM_Cubic, tangentMode=RCTM_User, time=0.8, value=1}, {interpMode=RCIM_Cubic, tangentMode=RCTM_User, time=1, value=0}]}}})
    )) {enabled=false, module=ScaleMeshSize005}
   &ScaleMeshSize(Scale Factor=&MakeVector(
     X=&RandomRangeFloat(
      Minimum=&FloatFromCurve(FloatCurve=DataInterface() {propertyValues {Curve {keys [
           {interpMode=RCIM_Cubic, tangentMode=RCTM_User, time=0, value=0.85}
           {interpMode=RCIM_Cubic, tangentMode=RCTM_User, time=0.45, value=1}
           <at 1 0.85 RCIM_Constant>
          ]}}})
      Maximum=&FloatFromCurve(FloatCurve=DataInterface() {propertyValues {Curve {keys [
           {interpMode=RCIM_Cubic, tangentMode=RCTM_User, time=0, value=0.85}
           {interpMode=RCIM_Cubic, tangentMode=RCTM_User, time=0.75, value=1}
           <at 1 0.85 RCIM_Constant>
          ]}}})
     )
     Y=&RandomRangeFloat(
      Minimum=&FloatFromCurve(FloatCurve=DataInterface() {propertyValues {Curve {keys [
           {interpMode=RCIM_Cubic, tangentMode=RCTM_User, time=0, value=0.85}
           {interpMode=RCIM_Cubic, tangentMode=RCTM_User, time=0.45, value=1}
           <at 1 0.85 RCIM_Constant>
          ]}}})
      Maximum=&FloatFromCurve(FloatCurve=DataInterface() {propertyValues {Curve {keys [
           {interpMode=RCIM_Cubic, tangentMode=RCTM_User, time=0, value=0.85}
           {interpMode=RCIM_Cubic, tangentMode=RCTM_User, time=0.75, value=1}
           <at 1 0.85 RCIM_Constant>
          ]}}})
     )
     Z=1
    )) {module=ScaleMeshSize002}
   &UpdateMeshOrientation(
    Rotation Rate=&RandomRangeFloat(Minimum=-0.1, Maximum=0.1)
    Rotation Vector=&RandomRangeVector(Minimum=<-1 -1 -1>, Maximum=<1 1 1>)
   )
   &ScaleColor(Scale RGB=&VectorFromFloat(Value=1), ScaleA=true, ScaleRGB=true)
   &GravityForce(Gravity=<0 0 -3>)
   &Drag(Drag=1.15, Use Linear Drag=true, Use Rotational Drag=true)
   &SolveForcesAndVelocity(Clamp Velocity=false, Limit Acceleration=false)
   &FN_Light_Attributes(
    Color=$Particles.Color
    Create Color Attribute=true
    Create Exponent Attribute=true
    Create Position Attribute=true
    Create Radius Attribute=true
    Enable Particle Light Enable=true
    Exponent=3
    Light Enabled=true
    Position:NiagaraPosition=$Particles.Position
    Radius=&FloatFromCurve(
     FloatCurve=DataInterface() {propertyValues {Curve {keys [
         {arriveTangent=0.3942, interpMode=RCIM_Cubic, leaveTangent=0.3942, tangentMode=RCTM_User, time=0, value=0}
         {arriveTangent=-2.5249, interpMode=RCIM_Cubic, leaveTangent=-2.5249, tangentMode=RCTM_User, time=0.1408, value=0.9847}
         {arriveTangent=-0.0405, interpMode=RCIM_Cubic, leaveTangent=-0.0405, tangentMode=RCTM_User, time=1, value=0}
        ]}}}
     Scale Curve=8
    )
    Scale Color=&Lerp_Float(
     A=10
     B=20
     Alpha=&FloatFromCurve(FloatCurve=DataInterface() {propertyValues {Curve {keys [
          {interpMode=RCIM_Cubic, tangentMode=RCTM_User, time=0, value=0}
          {arriveTangent=-1.9929, interpMode=RCIM_Cubic, leaveTangent=-1.9929, tangentMode=RCTM_User, time=0.5, value=1}
          {tangentMode=RCTM_User, time=1, value=0}
         ]}}})
    )
   )
   &CameraOffset(Camera Offset Amount=1)
  ]
  rendererClasses [&NiagaraMeshRendererProperties, &NiagaraLightRendererProperties]
  renderers [
   {rendererClass=&NiagaraMeshRendererProperties, rendererIndex=0}
   {rendererClass=&NiagaraLightRendererProperties, rendererIndex=1}
  ]
 }
 Flower_local001 {
  emitterUpdateScript [
   &EmitterState(
    Inactive Response=::NewEnumerator1
    Life Cycle Mode=::NewEnumerator1
    Loop Delay=&RandomRangeFloat(Minimum=0.5, Maximum=1)
    Loop Duration=10
    UseLoopDelay=false
   )
   &SpawnRate(Spawn Probability=0.9, Use Spawn Probability=true)
  ]
  particleSpawnScript [
   &PartitionParticles(Number Of Partitions=3, Particles Per Partition=1, Partition Mode:ENiagara_PartitionMode=::NewEnumerator1)
   /SetVariables(Particles.MeshIndex=$Particles.PartitionParticles.Partition) {isSetParameters=true, module=SetVariables_#2}
   &InitializeParticle(
    Color=&SelectLinearColorFromArray(
     Array Sampling Mode=::NewEnumerator0
     Direct Array Index=?
     Color Selection Array=DataInterface() {propertyValues {ColorData [
        <rgba 0.0275 0.5765 0.9961 1>
        <rgba 0.2968 0.1864 0.5089 1>
        <rgba 0.6939 0.104 0.6346 1>
        <rgba 0.1479 0.6172 0.6172 1>
        <rgba 1 0.324 0 1>
        <rgba 1 0.9258 0 1>
        <rgba 0.7969 0.0913 0.0913 1>
       ]}}
     Random Seed=0
    )
    Lifetime Max=5
    Lifetime Min=4
    Lifetime Mode=::NewEnumerator1
    Mesh Scale Mode=::NewEnumerator1
    Mesh Uniform Scale Max=0.35
    Mesh Uniform Scale Min=0.2
    Sprite Size Mode=::NewEnumerator1
    UsePositionOffset=false
    Write Material Random=true
   )
   &ShapeLocation(
    Box Size=<25 40 1>
    MaskSpawnGroup=false
    Override Seed=false
    Rotation Angle=90
    Rotation Axis=<0 1 0>
    Rotation Mode=::NewEnumerator1
    Shape Primitive=::NewEnumerator2
    Surface Only=false
    UseOwnerScale=true
   )
   &InitialMeshOrientation(Use Rotation Vector=false) {module=InitialMeshOrientation001}
   &DynamicMaterialParameters(
    Index 0 Param 1=1
    Index 0 Param 2=0.25
    Index 0 Param 3=0.5
    Module.Index0Param0Write=true
    Module.Index0Param1Write=true
    Module.Index0Param2Write=true
    Module.Index0Param3Write=true
   )
   &AddVelocity(
    Constrain To Radius=25
    Module.Use Falloff Distance=true
    Origin Offset=<0 -5 0>
    Override Seed=false
    Velocity Mode=::NewEnumerator1
    Velocity Speed=&RandomRangeFloat(Minimum=3, Maximum=6)
   ) {enabled=false}
  ]
  particleUpdateScript [
   &ParticleState()
   &ScaleMeshSize(Scale Factor=$Engine.Owner.Scale) {enabled=false, module=ScaleMeshSize006}
   &ScaleMeshSize(Scale Factor=&VectorFromFloat(Value=0.75)) {module=ScaleMeshSize003}
   &SkeletalMeshLocation(
    Bone / Socket Sampling Mode=::NewEnumerator5
    MaskSpawnGroup=false
    Mesh Sampling Type=::NewEnumerator6
    Orientation Sampling=::NewEnumerator1
    Sampled Position Offset=&AddVectorToPosition(
     Position=$Particles.ShapeLocation.ShapePosition
     Vector=&RandomRangeVector(Minimum=<0 1 0>, Maximum=<0 2 0>)
    )
    Sampled Position Offset Coordinate Space:ENiagaraSkelMeshTransforms=::NewEnumerator1
    Skeletal Mesh=$System.Skeletal Mesh
   )
   &CurlNoiseForce(Noise Strength=5, Pan Noise=true, Pan Noise Field=<0 0 4>)
   &ScaleMeshSize(Scale Factor=&VectorFromFloat(Value=&Lerp_Float(
      A=1
      B=1.7
      Alpha=&FloatFromCurve(
       FloatCurve=DataInterface() {propertyValues {Curve {keys [
           {arriveTangent=9.0725, interpMode=RCIM_Cubic, time=0, value=0}
           <at 0.1 1 RCIM_Cubic>
           {arriveTangent=-0.0088, interpMode=RCIM_Cubic, leaveTangent=-0.0088, tangentMode=RCTM_User, time=1, value=0}
          ]}}}
       CurveIndex=$NPC.NPC_Sparks_Audio_Parameters001.BeatLifetime
      )
     ))) {module=ScaleMeshSize001}
   &DynamicMaterialParameters(
    Index 0 Param 1=1
    Index 0 Param 2=0.25
    Index 0 Param 3=0.5
    Index 1 Param 1=&FloatFromCurve(
     FloatCurve=DataInterface() {propertyValues {Curve {keys [<at 0 1>,<at 1 0>]}}}
     Scale Curve=&RandomRangeFloat(Minimum=0.25, Maximum=0.5)
    )
    Module.Index0Param0Write=true
    Module.Index0Param1Write=true
    Module.Index0Param2Write=true
    Module.Index0Param3Write=true
    Module.Index1Param0Write=true
    Module.Index1Param1Write=false
    Module.Index1Param2Write=false
    Module.Index1Param3Write=false
    Write Parameter Index 1=true
   ) {module=DynamicMaterialParameters001}
   &ScaleMeshSize(Scale Factor=&VectorFromFloat(Value=&FloatFromCurve(FloatCurve=DataInterface() {propertyValues {Curve {keys [
          {arriveTangent=0.3651, interpMode=RCIM_Cubic, leaveTangent=0.3651, tangentMode=RCTM_User, time=0, value=0}
          {arriveTangent=-2.5249, leaveTangent=-2.5249, time=0.1408, value=0.9847}
          {arriveTangent=-1.146, interpMode=RCIM_Cubic, leaveTangent=-1.146, time=0.6, value=1}
          {arriveTangent=-0.0405, interpMode=RCIM_Cubic, leaveTangent=-0.0405, tangentMode=RCTM_User, time=1, value=0}
         ]}}})))
   &ScaleMeshSize(Scale Factor=&MakeVector(
     X=&FloatFromCurve(FloatCurve=DataInterface() {propertyValues {Curve {keys [
          {arriveTangent=1.8704, interpMode=RCIM_Cubic, leaveTangent=1.8704, tangentMode=RCTM_User, time=0, value=0}
          {arriveTangent=0.2799, leaveTangent=1.1179, tangentMode=RCTM_User, time=0.3, value=1}
          {arriveTangent=0.0224, interpMode=RCIM_Cubic, leaveTangent=0.0224, tangentMode=RCTM_User, time=1, value=1}
         ]}}})
     Y=&FloatFromCurve(FloatCurve=DataInterface() {propertyValues {Curve {keys [
          {arriveTangent=2.4662, interpMode=RCIM_Cubic, leaveTangent=2.4662, tangentMode=RCTM_User, time=0, value=0}
          {arriveTangent=0.422, leaveTangent=1.1179, tangentMode=RCTM_User, time=0.3, value=1}
          {arriveTangent=0.0224, interpMode=RCIM_Cubic, leaveTangent=0.0224, tangentMode=RCTM_User, time=1, value=1}
         ]}}})
     Z=&FloatFromCurve(FloatCurve=DataInterface() {propertyValues {Curve {keys [
          {interpMode=RCIM_Cubic, tangentMode=RCTM_User, time=0, value=0}
          {arriveTangent=0.3068, leaveTangent=0.3068, time=0.2, value=1.025}
          {arriveTangent=0.0743, interpMode=RCIM_Cubic, leaveTangent=0.0743, tangentMode=RCTM_User, time=1, value=1}
         ]}}})
    )) {enabled=false, module=ScaleMeshSize004}
   &ScaleMeshSize(Scale Factor=&MakeVector(
     X=&FloatFromCurve(FloatCurve=DataInterface() {propertyValues {Curve {keys [
          {tangentMode=RCTM_User, time=0, value=1}
          {arriveTangent=-3.4324, interpMode=RCIM_Cubic, leaveTangent=-3.4324, tangentMode=RCTM_User, time=0.7, value=1}
          {arriveTangent=-0.4311, interpMode=RCIM_Cubic, leaveTangent=-0.4311, tangentMode=RCTM_User, time=1, value=0}
         ]}}})
     Y=&FloatFromCurve(FloatCurve=DataInterface() {propertyValues {Curve {keys [
          {tangentMode=RCTM_User, time=0, value=1}
          {arriveTangent=-3.3527, interpMode=RCIM_Cubic, leaveTangent=-3.3527, tangentMode=RCTM_User, time=0.7, value=1}
          {arriveTangent=-0.4311, interpMode=RCIM_Cubic, leaveTangent=-0.4311, tangentMode=RCTM_User, time=1, value=0}
         ]}}})
     Z=&FloatFromCurve(FloatCurve=DataInterface() {propertyValues {Curve {keys [{interpMode=RCIM_Cubic, tangentMode=RCTM_User, time=0.8, value=1}, {interpMode=RCIM_Cubic, tangentMode=RCTM_User, time=1, value=0}]}}})
    )) {enabled=false, module=ScaleMeshSize005}
   &ScaleMeshSize(Scale Factor=&MakeVector(
     X=&RandomRangeFloat(
      Minimum=&FloatFromCurve(FloatCurve=DataInterface() {propertyValues {Curve {keys [
           {interpMode=RCIM_Cubic, tangentMode=RCTM_User, time=0, value=0.85}
           {interpMode=RCIM_Cubic, tangentMode=RCTM_User, time=0.45, value=1}
           <at 1 0.85 RCIM_Constant>
          ]}}})
      Maximum=&FloatFromCurve(FloatCurve=DataInterface() {propertyValues {Curve {keys [
           {interpMode=RCIM_Cubic, tangentMode=RCTM_User, time=0, value=0.85}
           {interpMode=RCIM_Cubic, tangentMode=RCTM_User, time=0.75, value=1}
           <at 1 0.85 RCIM_Constant>
          ]}}})
     )
     Y=&RandomRangeFloat(
      Minimum=&FloatFromCurve(FloatCurve=DataInterface() {propertyValues {Curve {keys [
           {interpMode=RCIM_Cubic, tangentMode=RCTM_User, time=0, value=0.85}
           {interpMode=RCIM_Cubic, tangentMode=RCTM_User, time=0.45, value=1}
           <at 1 0.85 RCIM_Constant>
          ]}}})
      Maximum=&FloatFromCurve(FloatCurve=DataInterface() {propertyValues {Curve {keys [
           {interpMode=RCIM_Cubic, tangentMode=RCTM_User, time=0, value=0.85}
           {interpMode=RCIM_Cubic, tangentMode=RCTM_User, time=0.75, value=1}
           <at 1 0.85 RCIM_Constant>
          ]}}})
     )
     Z=1
    )) {module=ScaleMeshSize002}
   &UpdateMeshOrientation(
    Rotation Rate=&RandomRangeFloat(Minimum=-0.1, Maximum=0.1)
    Rotation Vector=&RandomRangeVector(Minimum=<-1 -1 -1>, Maximum=<1 1 1>)
   )
   &ScaleColor(Scale RGB=&VectorFromFloat(Value=1), ScaleA=true, ScaleRGB=true)
   &GravityForce(Gravity=<0 0 -3>) {enabled=false}
   /SetVariables(Particles.Position=&AddVectorToPosition(Position=$Particles.Position, Vector=&MakeVector(X=0, Y=0, Z=0.03))) {isSetParameters=true}
   &Drag(Drag=1.15, Use Linear Drag=true, Use Rotational Drag=true) {enabled=false}
   &SolveForcesAndVelocity(Clamp Velocity=false, Limit Acceleration=false)
   &FN_Light_Attributes(
    Color=$Particles.Color
    Create Color Attribute=true
    Create Exponent Attribute=true
    Create Position Attribute=true
    Create Radius Attribute=true
    Enable Particle Light Enable=true
    Exponent=3
    Light Enabled=true
    Position:NiagaraPosition=$Particles.Position
    Radius=&FloatFromCurve(
     FloatCurve=DataInterface() {propertyValues {Curve {keys [
         {arriveTangent=0.3942, interpMode=RCIM_Cubic, leaveTangent=0.3942, tangentMode=RCTM_User, time=0, value=0}
         {arriveTangent=-2.5249, interpMode=RCIM_Cubic, leaveTangent=-2.5249, tangentMode=RCTM_User, time=0.1408, value=0.9847}
         {arriveTangent=-0.0405, interpMode=RCIM_Cubic, leaveTangent=-0.0405, tangentMode=RCTM_User, time=1, value=0}
        ]}}}
     Scale Curve=8
    )
    Scale Color=&Lerp_Float(
     A=10
     B=20
     Alpha=&FloatFromCurve(FloatCurve=DataInterface() {propertyValues {Curve {keys [
          {interpMode=RCIM_Cubic, tangentMode=RCTM_User, time=0, value=0}
          {arriveTangent=-1.9929, interpMode=RCIM_Cubic, leaveTangent=-1.9929, tangentMode=RCTM_User, time=0.5, value=1}
          {tangentMode=RCTM_User, time=1, value=0}
         ]}}})
    )
   )
   &CameraOffset(Camera Offset Amount=1)
  ]
  rendererClasses [&NiagaraMeshRendererProperties, &NiagaraLightRendererProperties]
  renderers [
   {rendererClass=&NiagaraMeshRendererProperties, rendererIndex=0}
   {rendererClass=&NiagaraLightRendererProperties, rendererIndex=1}
  ]
 }
 Flower_local_InitialBurst {
  bEnabled=false
  emitterUpdateScript [
   &EmitterState(
    Inactive Response=::NewEnumerator1
    Life Cycle Mode=::NewEnumerator1
    Loop Behavior=::NewEnumerator1
    Loop Delay=&RandomRangeFloat(Minimum=0.5, Maximum=1)
    Loop Duration=10
    UseLoopDelay=false
   )
   &SpawnBurst_Instantaneous(Spawn Count=4, Use Spawn Probability=false, UseLoopCountLimit=false)
   &SpawnRate(Spawn Probability=0.9, Use Spawn Probability=true) {enabled=false}
  ]
  particleSpawnScript [
   &PartitionParticles(Number Of Partitions=3, Particles Per Partition=1, Partition Mode:ENiagara_PartitionMode=::NewEnumerator1)
   /SetVariables(Particles.MeshIndex=$Particles.PartitionParticles.Partition) {isSetParameters=true, module=SetVariables_#2}
   &InitializeParticle(
    Color=&SelectLinearColorFromArray(
     Array Sampling Mode=::NewEnumerator0
     Direct Array Index=?
     Color Selection Array=DataInterface() {propertyValues {ColorData [
        <rgba 0.0275 0.5765 0.9961 1>
        <rgba 0.2968 0.1864 0.5089 1>
        <rgba 0.6939 0.104 0.6346 1>
        <rgba 0.1479 0.6172 0.6172 1>
        <rgba 1 0.324 0 1>
        <rgba 1 0.9258 0 1>
        <rgba 0.7969 0.0913 0.0913 1>
       ]}}
     Random Seed=0
    )
    Lifetime Max=5
    Lifetime Min=4
    Lifetime Mode=::NewEnumerator1
    Mesh Scale Mode=::NewEnumerator1
    Mesh Uniform Scale Max=0.35
    Mesh Uniform Scale Min=0.2
    Sprite Size Mode=::NewEnumerator1
    UsePositionOffset=false
    Write Material Random=true
   )
   &ShapeLocation(
    Box Size=<25 40 1>
    MaskSpawnGroup=false
    Override Seed=false
    Rotation Angle=90
    Rotation Axis=<0 1 0>
    Rotation Mode=::NewEnumerator1
    Shape Primitive=::NewEnumerator2
    Surface Only=false
    UseOwnerScale=true
   )
   &InitialMeshOrientation(Use Rotation Vector=false) {module=InitialMeshOrientation001}
   &DynamicMaterialParameters(
    Index 0 Param 1=1
    Index 0 Param 2=0.25
    Index 0 Param 3=0.5
    Module.Index0Param0Write=true
    Module.Index0Param1Write=true
    Module.Index0Param2Write=true
    Module.Index0Param3Write=true
   )
   &AddVelocity(
    Constrain To Radius=25
    Module.Use Falloff Distance=true
    Origin Offset=<0 -5 0>
    Override Seed=false
    Velocity Mode=::NewEnumerator1
    Velocity Speed=&RandomRangeFloat(Minimum=3, Maximum=6)
   )
  ]
  particleUpdateScript [
   &ParticleState()
   &ScaleMeshSize(Scale Factor=&VectorFromFloat(Value=0.75)) {module=ScaleMeshSize003}
   &SkeletalMeshLocation(
    Bone / Socket Sampling Mode=::NewEnumerator5
    MaskSpawnGroup=false
    Mesh Sampling Type=::NewEnumerator6
    Orientation Sampling=::NewEnumerator1
    Position Sampling=::NewEnumerator1
    Sampled Position Offset=&AddVectorToPosition(
     Position=$Particles.ShapeLocation.ShapePosition
     Vector=&RandomRangeVector(Minimum=<0 1 0>, Maximum=<0 2 0>)
    )
    Sampled Position Offset Coordinate Space:ENiagaraSkelMeshTransforms=::NewEnumerator1
    Skeletal Mesh=$System.Skeletal Mesh
   )
   &CurlNoiseForce(Noise Strength=5, Pan Noise=true, Pan Noise Field=<0 0 4>)
   &ScaleMeshSize(Scale Factor=&VectorFromFloat(Value=&Lerp_Float(
      A=1
      B=1.7
      Alpha=&FloatFromCurve(
       FloatCurve=DataInterface() {propertyValues {Curve {keys [
           {arriveTangent=9.0725, interpMode=RCIM_Cubic, time=0, value=0}
           <at 0.1 1 RCIM_Cubic>
           {arriveTangent=-0.0088, interpMode=RCIM_Cubic, leaveTangent=-0.0088, tangentMode=RCTM_User, time=1, value=0}
          ]}}}
       CurveIndex=$NPC.NPC_Sparks_Audio_Parameters001.BeatLifetime
      )
     ))) {module=ScaleMeshSize001}
   &DynamicMaterialParameters(
    Index 0 Param 1=1
    Index 0 Param 2=0.25
    Index 0 Param 3=0.5
    Index 1 Param 1=&FloatFromCurve(
     FloatCurve=DataInterface() {propertyValues {Curve {keys [
         {arriveTangent=9.0725, interpMode=RCIM_Cubic, leaveTangent=9.0725, tangentMode=RCTM_User, time=0, value=0}
         {arriveTangent=-2.5249, interpMode=RCIM_Cubic, leaveTangent=-2.5249, tangentMode=RCTM_User, time=0.1408, value=0.9847}
         {arriveTangent=-0.0405, interpMode=RCIM_Cubic, leaveTangent=-0.0405, tangentMode=RCTM_User, time=1, value=0}
        ]}}}
     Scale Curve=&RandomRangeFloat(Minimum=0.25, Maximum=0.5)
    )
    Module.Index0Param0Write=true
    Module.Index0Param1Write=true
    Module.Index0Param2Write=true
    Module.Index0Param3Write=true
    Module.Index1Param0Write=true
    Module.Index1Param1Write=false
    Module.Index1Param2Write=false
    Module.Index1Param3Write=false
    Write Parameter Index 1=true
   ) {module=DynamicMaterialParameters001}
   &ScaleMeshSize(Scale Factor=&VectorFromFloat(Value=&FloatFromCurve(FloatCurve=DataInterface() {propertyValues {Curve {keys [
          {arriveTangent=0.3651, interpMode=RCIM_Cubic, leaveTangent=0.3651, tangentMode=RCTM_User, time=0, value=0}
          {arriveTangent=-2.5249, leaveTangent=-2.5249, time=0.1408, value=0.9847}
          {arriveTangent=-1.146, interpMode=RCIM_Cubic, leaveTangent=-1.146, time=0.6, value=1}
          {arriveTangent=-0.0405, interpMode=RCIM_Cubic, leaveTangent=-0.0405, tangentMode=RCTM_User, time=1, value=0}
         ]}}})))
   &ScaleMeshSize(Scale Factor=&MakeVector(
     X=&FloatFromCurve(FloatCurve=DataInterface() {propertyValues {Curve {keys [
          {arriveTangent=1.8704, interpMode=RCIM_Cubic, leaveTangent=1.8704, tangentMode=RCTM_User, time=0, value=0}
          {arriveTangent=0.2799, leaveTangent=1.1179, tangentMode=RCTM_User, time=0.3, value=1}
          {arriveTangent=0.0224, interpMode=RCIM_Cubic, leaveTangent=0.0224, tangentMode=RCTM_User, time=1, value=1}
         ]}}})
     Y=&FloatFromCurve(FloatCurve=DataInterface() {propertyValues {Curve {keys [
          {arriveTangent=2.4662, interpMode=RCIM_Cubic, leaveTangent=2.4662, tangentMode=RCTM_User, time=0, value=0}
          {arriveTangent=0.422, leaveTangent=1.1179, tangentMode=RCTM_User, time=0.3, value=1}
          {arriveTangent=0.0224, interpMode=RCIM_Cubic, leaveTangent=0.0224, tangentMode=RCTM_User, time=1, value=1}
         ]}}})
     Z=&FloatFromCurve(FloatCurve=DataInterface() {propertyValues {Curve {keys [
          {interpMode=RCIM_Cubic, tangentMode=RCTM_User, time=0, value=0}
          {arriveTangent=0.3068, leaveTangent=0.3068, time=0.2, value=1.025}
          {arriveTangent=0.0743, interpMode=RCIM_Cubic, leaveTangent=0.0743, tangentMode=RCTM_User, time=1, value=1}
         ]}}})
    )) {enabled=false, module=ScaleMeshSize004}
   &ScaleMeshSize(Scale Factor=&MakeVector(
     X=&FloatFromCurve(FloatCurve=DataInterface() {propertyValues {Curve {keys [
          {tangentMode=RCTM_User, time=0, value=1}
          {arriveTangent=-3.4324, interpMode=RCIM_Cubic, leaveTangent=-3.4324, tangentMode=RCTM_User, time=0.7, value=1}
          {arriveTangent=-0.4311, interpMode=RCIM_Cubic, leaveTangent=-0.4311, tangentMode=RCTM_User, time=1, value=0}
         ]}}})
     Y=&FloatFromCurve(FloatCurve=DataInterface() {propertyValues {Curve {keys [
          {tangentMode=RCTM_User, time=0, value=1}
          {arriveTangent=-3.3527, interpMode=RCIM_Cubic, leaveTangent=-3.3527, tangentMode=RCTM_User, time=0.7, value=1}
          {arriveTangent=-0.4311, interpMode=RCIM_Cubic, leaveTangent=-0.4311, tangentMode=RCTM_User, time=1, value=0}
         ]}}})
     Z=&FloatFromCurve(FloatCurve=DataInterface() {propertyValues {Curve {keys [{interpMode=RCIM_Cubic, tangentMode=RCTM_User, time=0.8, value=1}, {interpMode=RCIM_Cubic, tangentMode=RCTM_User, time=1, value=0}]}}})
    )) {enabled=false, module=ScaleMeshSize005}
   &ScaleMeshSize(Scale Factor=&MakeVector(
     X=&RandomRangeFloat(
      Minimum=&FloatFromCurve(FloatCurve=DataInterface() {propertyValues {Curve {keys [
           {interpMode=RCIM_Cubic, tangentMode=RCTM_User, time=0, value=0.85}
           {interpMode=RCIM_Cubic, tangentMode=RCTM_User, time=0.45, value=1}
           <at 1 0.85 RCIM_Constant>
          ]}}})
      Maximum=&FloatFromCurve(FloatCurve=DataInterface() {propertyValues {Curve {keys [
           {interpMode=RCIM_Cubic, tangentMode=RCTM_User, time=0, value=0.85}
           {interpMode=RCIM_Cubic, tangentMode=RCTM_User, time=0.75, value=1}
           <at 1 0.85 RCIM_Constant>
          ]}}})
     )
     Y=&RandomRangeFloat(
      Minimum=&FloatFromCurve(FloatCurve=DataInterface() {propertyValues {Curve {keys [
           {interpMode=RCIM_Cubic, tangentMode=RCTM_User, time=0, value=0.85}
           {interpMode=RCIM_Cubic, tangentMode=RCTM_User, time=0.45, value=1}
           <at 1 0.85 RCIM_Constant>
          ]}}})
      Maximum=&FloatFromCurve(FloatCurve=DataInterface() {propertyValues {Curve {keys [
           {interpMode=RCIM_Cubic, tangentMode=RCTM_User, time=0, value=0.85}
           {interpMode=RCIM_Cubic, tangentMode=RCTM_User, time=0.75, value=1}
           <at 1 0.85 RCIM_Constant>
          ]}}})
     )
     Z=1
    )) {module=ScaleMeshSize002}
   &UpdateMeshOrientation(
    Rotation Rate=&RandomRangeFloat(Minimum=-0.1, Maximum=0.1)
    Rotation Vector=&RandomRangeVector(Minimum=<-1 -1 -1>, Maximum=<1 1 1>)
   )
   &ScaleColor(Scale RGB=&VectorFromFloat(Value=1), ScaleA=true, ScaleRGB=true)
   &GravityForce(Gravity=<0 0 -3>)
   &Drag(Drag=1.15, Use Linear Drag=true, Use Rotational Drag=true)
   &SolveForcesAndVelocity(Clamp Velocity=false, Limit Acceleration=false)
   &FN_Light_Attributes(
    Color=$Particles.Color
    Create Color Attribute=true
    Create Exponent Attribute=true
    Create Position Attribute=true
    Create Radius Attribute=true
    Enable Particle Light Enable=true
    Exponent=3
    Light Enabled=true
    Position:NiagaraPosition=$Particles.Position
    Radius=&FloatFromCurve(
     FloatCurve=DataInterface() {propertyValues {Curve {keys [
         {arriveTangent=0.3942, interpMode=RCIM_Cubic, leaveTangent=0.3942, tangentMode=RCTM_User, time=0, value=0}
         {arriveTangent=-2.5249, interpMode=RCIM_Cubic, leaveTangent=-2.5249, tangentMode=RCTM_User, time=0.1408, value=0.9847}
         {arriveTangent=-0.0405, interpMode=RCIM_Cubic, leaveTangent=-0.0405, tangentMode=RCTM_User, time=1, value=0}
        ]}}}
     Scale Curve=8
    )
    Scale Color=&Lerp_Float(
     A=10
     B=20
     Alpha=&FloatFromCurve(FloatCurve=DataInterface() {propertyValues {Curve {keys [
          {interpMode=RCIM_Cubic, tangentMode=RCTM_User, time=0, value=0}
          {arriveTangent=-1.9929, interpMode=RCIM_Cubic, leaveTangent=-1.9929, tangentMode=RCTM_User, time=0.5, value=1}
          {tangentMode=RCTM_User, time=1, value=0}
         ]}}})
    )
   )
   &CameraOffset(Camera Offset Amount=1)
  ]
  rendererClasses [&NiagaraMeshRendererProperties, &NiagaraLightRendererProperties]
  renderers [
   {rendererClass=&NiagaraMeshRendererProperties, rendererIndex=0}
   {rendererClass=&NiagaraLightRendererProperties, rendererIndex=1}
  ]
 }
 Flower_local_InitialBurst001 {
  emitterUpdateScript [
   &EmitterState(
    Inactive Response=::NewEnumerator1
    Life Cycle Mode=::NewEnumerator1
    Loop Behavior=::NewEnumerator1
    Loop Delay=&RandomRangeFloat(Minimum=0.5, Maximum=1)
    Loop Duration=10
    UseLoopDelay=false
   )
   &SpawnBurst_Instantaneous(Spawn Count=4, Use Spawn Probability=false, UseLoopCountLimit=false)
   &SpawnRate(Spawn Probability=0.9, Use Spawn Probability=true) {enabled=false}
  ]
  particleSpawnScript [
   &PartitionParticles(Number Of Partitions=3, Particles Per Partition=1, Partition Mode:ENiagara_PartitionMode=::NewEnumerator1)
   /SetVariables(Particles.MeshIndex=$Particles.PartitionParticles.Partition) {isSetParameters=true, module=SetVariables_#2}
   &InitializeParticle(
    Color=&SelectLinearColorFromArray(
     Array Sampling Mode=::NewEnumerator0
     Direct Array Index=?
     Color Selection Array=DataInterface() {propertyValues {ColorData [
        <rgba 0.0275 0.5765 0.9961 1>
        <rgba 0.2968 0.1864 0.5089 1>
        <rgba 0.6939 0.104 0.6346 1>
        <rgba 0.1479 0.6172 0.6172 1>
        <rgba 1 0.324 0 1>
        <rgba 1 0.9258 0 1>
        <rgba 0.7969 0.0913 0.0913 1>
       ]}}
     Random Seed=0
    )
    Lifetime Max=5
    Lifetime Min=4
    Lifetime Mode=::NewEnumerator1
    Mesh Scale Mode=::NewEnumerator1
    Mesh Uniform Scale Max=0.35
    Mesh Uniform Scale Min=0.2
    Sprite Size Mode=::NewEnumerator1
    UsePositionOffset=false
    Write Material Random=true
   )
   &ShapeLocation(
    Box Size=<25 40 1>
    MaskSpawnGroup=false
    Override Seed=false
    Rotation Angle=90
    Rotation Axis=<0 1 0>
    Rotation Mode=::NewEnumerator1
    Shape Primitive=::NewEnumerator2
    Surface Only=false
    UseOwnerScale=true
   )
   &InitialMeshOrientation(Use Rotation Vector=false) {module=InitialMeshOrientation001}
   &DynamicMaterialParameters(
    Index 0 Param 1=1
    Index 0 Param 2=0.25
    Index 0 Param 3=0.5
    Module.Index0Param0Write=true
    Module.Index0Param1Write=true
    Module.Index0Param2Write=true
    Module.Index0Param3Write=true
   )
   &AddVelocity(
    Constrain To Radius=25
    Module.Use Falloff Distance=true
    Origin Offset=<0 -5 0>
    Override Seed=false
    Velocity Mode=::NewEnumerator1
    Velocity Speed=&RandomRangeFloat(Minimum=3, Maximum=6)
   ) {enabled=false}
  ]
  particleUpdateScript [
   &ParticleState()
   &ScaleMeshSize(Scale Factor=$Engine.Owner.Scale) {enabled=false, module=ScaleMeshSize006}
   &ScaleMeshSize(Scale Factor=&VectorFromFloat(Value=0.75)) {module=ScaleMeshSize003}
   &SkeletalMeshLocation(
    Bone / Socket Sampling Mode=::NewEnumerator5
    MaskSpawnGroup=false
    Mesh Sampling Type=::NewEnumerator6
    Orientation Sampling=::NewEnumerator1
    Sampled Position Offset=&AddVectorToPosition(
     Position=$Particles.ShapeLocation.ShapePosition
     Vector=&RandomRangeVector(Minimum=<0 1 0>, Maximum=<0 2 0>)
    )
    Sampled Position Offset Coordinate Space:ENiagaraSkelMeshTransforms=::NewEnumerator1
    Skeletal Mesh=$System.Skeletal Mesh
   )
   &CurlNoiseForce(Noise Strength=5, Pan Noise=true, Pan Noise Field=<0 0 4>)
   &ScaleMeshSize(Scale Factor=&VectorFromFloat(Value=&Lerp_Float(
      A=1
      B=1.7
      Alpha=&FloatFromCurve(
       FloatCurve=DataInterface() {propertyValues {Curve {keys [
           {arriveTangent=9.0725, interpMode=RCIM_Cubic, time=0, value=0}
           <at 0.1 1 RCIM_Cubic>
           {arriveTangent=-0.0088, interpMode=RCIM_Cubic, leaveTangent=-0.0088, tangentMode=RCTM_User, time=1, value=0}
          ]}}}
       CurveIndex=$NPC.NPC_Sparks_Audio_Parameters001.BeatLifetime
      )
     ))) {module=ScaleMeshSize001}
   &DynamicMaterialParameters(
    Index 0 Param 1=1
    Index 0 Param 2=0.25
    Index 0 Param 3=0.5
    Index 1 Param 1=&FloatFromCurve(
     FloatCurve=DataInterface() {propertyValues {Curve {keys [
         {arriveTangent=9.0725, interpMode=RCIM_Cubic, leaveTangent=9.0725, tangentMode=RCTM_User, time=0, value=0}
         {arriveTangent=-2.5249, interpMode=RCIM_Cubic, leaveTangent=-2.5249, tangentMode=RCTM_User, time=0.1408, value=0.9847}
         {arriveTangent=-0.0405, interpMode=RCIM_Cubic, leaveTangent=-0.0405, tangentMode=RCTM_User, time=1, value=0}
        ]}}}
     Scale Curve=&RandomRangeFloat(Minimum=0.25, Maximum=0.5)
    )
    Module.Index0Param0Write=true
    Module.Index0Param1Write=true
    Module.Index0Param2Write=true
    Module.Index0Param3Write=true
    Module.Index1Param0Write=true
    Module.Index1Param1Write=false
    Module.Index1Param2Write=false
    Module.Index1Param3Write=false
    Write Parameter Index 1=true
   ) {module=DynamicMaterialParameters001}
   &ScaleMeshSize(Scale Factor=&VectorFromFloat(Value=&FloatFromCurve(FloatCurve=DataInterface() {propertyValues {Curve {keys [
          {arriveTangent=0.3651, interpMode=RCIM_Cubic, leaveTangent=0.3651, tangentMode=RCTM_User, time=0, value=0}
          {arriveTangent=-2.5249, leaveTangent=-2.5249, time=0.1408, value=0.9847}
          {arriveTangent=-1.146, interpMode=RCIM_Cubic, leaveTangent=-1.146, time=0.6, value=1}
          {arriveTangent=-0.0405, interpMode=RCIM_Cubic, leaveTangent=-0.0405, tangentMode=RCTM_User, time=1, value=0}
         ]}}})))
   &ScaleMeshSize(Scale Factor=&MakeVector(
     X=&FloatFromCurve(FloatCurve=DataInterface() {propertyValues {Curve {keys [
          {arriveTangent=1.8704, interpMode=RCIM_Cubic, leaveTangent=1.8704, tangentMode=RCTM_User, time=0, value=0}
          {arriveTangent=0.2799, leaveTangent=1.1179, tangentMode=RCTM_User, time=0.3, value=1}
          {arriveTangent=0.0224, interpMode=RCIM_Cubic, leaveTangent=0.0224, tangentMode=RCTM_User, time=1, value=1}
         ]}}})
     Y=&FloatFromCurve(FloatCurve=DataInterface() {propertyValues {Curve {keys [
          {arriveTangent=2.4662, interpMode=RCIM_Cubic, leaveTangent=2.4662, tangentMode=RCTM_User, time=0, value=0}
          {arriveTangent=0.422, leaveTangent=1.1179, tangentMode=RCTM_User, time=0.3, value=1}
          {arriveTangent=0.0224, interpMode=RCIM_Cubic, leaveTangent=0.0224, tangentMode=RCTM_User, time=1, value=1}
         ]}}})
     Z=&FloatFromCurve(FloatCurve=DataInterface() {propertyValues {Curve {keys [
          {interpMode=RCIM_Cubic, tangentMode=RCTM_User, time=0, value=0}
          {arriveTangent=0.3068, leaveTangent=0.3068, time=0.2, value=1.025}
          {arriveTangent=0.0743, interpMode=RCIM_Cubic, leaveTangent=0.0743, tangentMode=RCTM_User, time=1, value=1}
         ]}}})
    )) {enabled=false, module=ScaleMeshSize004}
   &ScaleMeshSize(Scale Factor=&MakeVector(
     X=&FloatFromCurve(FloatCurve=DataInterface() {propertyValues {Curve {keys [
          {tangentMode=RCTM_User, time=0, value=1}
          {arriveTangent=-3.4324, interpMode=RCIM_Cubic, leaveTangent=-3.4324, tangentMode=RCTM_User, time=0.7, value=1}
          {arriveTangent=-0.4311, interpMode=RCIM_Cubic, leaveTangent=-0.4311, tangentMode=RCTM_User, time=1, value=0}
         ]}}})
     Y=&FloatFromCurve(FloatCurve=DataInterface() {propertyValues {Curve {keys [
          {tangentMode=RCTM_User, time=0, value=1}
          {arriveTangent=-3.3527, interpMode=RCIM_Cubic, leaveTangent=-3.3527, tangentMode=RCTM_User, time=0.7, value=1}
          {arriveTangent=-0.4311, interpMode=RCIM_Cubic, leaveTangent=-0.4311, tangentMode=RCTM_User, time=1, value=0}
         ]}}})
     Z=&FloatFromCurve(FloatCurve=DataInterface() {propertyValues {Curve {keys [{interpMode=RCIM_Cubic, tangentMode=RCTM_User, time=0.8, value=1}, {interpMode=RCIM_Cubic, tangentMode=RCTM_User, time=1, value=0}]}}})
    )) {enabled=false, module=ScaleMeshSize005}
   &ScaleMeshSize(Scale Factor=&MakeVector(
     X=&RandomRangeFloat(
      Minimum=&FloatFromCurve(FloatCurve=DataInterface() {propertyValues {Curve {keys [
           {interpMode=RCIM_Cubic, tangentMode=RCTM_User, time=0, value=0.85}
           {interpMode=RCIM_Cubic, tangentMode=RCTM_User, time=0.45, value=1}
           <at 1 0.85 RCIM_Constant>
          ]}}})
      Maximum=&FloatFromCurve(FloatCurve=DataInterface() {propertyValues {Curve {keys [
           {interpMode=RCIM_Cubic, tangentMode=RCTM_User, time=0, value=0.85}
           {interpMode=RCIM_Cubic, tangentMode=RCTM_User, time=0.75, value=1}
           <at 1 0.85 RCIM_Constant>
          ]}}})
     )
     Y=&RandomRangeFloat(
      Minimum=&FloatFromCurve(FloatCurve=DataInterface() {propertyValues {Curve {keys [
           {interpMode=RCIM_Cubic, tangentMode=RCTM_User, time=0, value=0.85}
           {interpMode=RCIM_Cubic, tangentMode=RCTM_User, time=0.45, value=1}
           <at 1 0.85 RCIM_Constant>
          ]}}})
      Maximum=&FloatFromCurve(FloatCurve=DataInterface() {propertyValues {Curve {keys [
           {interpMode=RCIM_Cubic, tangentMode=RCTM_User, time=0, value=0.85}
           {interpMode=RCIM_Cubic, tangentMode=RCTM_User, time=0.75, value=1}
           <at 1 0.85 RCIM_Constant>
          ]}}})
     )
     Z=1
    )) {module=ScaleMeshSize002}
   &UpdateMeshOrientation(
    Rotation Rate=&RandomRangeFloat(Minimum=-0.1, Maximum=0.1)
    Rotation Vector=&RandomRangeVector(Minimum=<-1 -1 -1>, Maximum=<1 1 1>)
   )
   &ScaleColor(Scale RGB=&VectorFromFloat(Value=1), ScaleA=true, ScaleRGB=true)
   &GravityForce(Gravity=<0 0 -3>) {enabled=false}
   /SetVariables(Particles.Position=&AddVectorToPosition(Position=$Particles.Position, Vector=&MakeVector(X=0, Y=0, Z=-0.03))) {enabled=false, isSetParameters=true}
   &Drag(Drag=1.15, Use Linear Drag=true, Use Rotational Drag=true) {enabled=false}
   &SolveForcesAndVelocity(Clamp Velocity=false, Limit Acceleration=false)
   &FN_Light_Attributes(
    Color=$Particles.Color
    Create Color Attribute=true
    Create Exponent Attribute=true
    Create Position Attribute=true
    Create Radius Attribute=true
    Enable Particle Light Enable=true
    Exponent=3
    Light Enabled=true
    Position:NiagaraPosition=$Particles.Position
    Radius=&FloatFromCurve(
     FloatCurve=DataInterface() {propertyValues {Curve {keys [
         {arriveTangent=0.3942, interpMode=RCIM_Cubic, leaveTangent=0.3942, tangentMode=RCTM_User, time=0, value=0}
         {arriveTangent=-2.5249, interpMode=RCIM_Cubic, leaveTangent=-2.5249, tangentMode=RCTM_User, time=0.1408, value=0.9847}
         {arriveTangent=-0.0405, interpMode=RCIM_Cubic, leaveTangent=-0.0405, tangentMode=RCTM_User, time=1, value=0}
        ]}}}
     Scale Curve=8
    )
    Scale Color=&Lerp_Float(
     A=10
     B=20
     Alpha=&FloatFromCurve(FloatCurve=DataInterface() {propertyValues {Curve {keys [
          {interpMode=RCIM_Cubic, tangentMode=RCTM_User, time=0, value=0}
          {arriveTangent=-1.9929, interpMode=RCIM_Cubic, leaveTangent=-1.9929, tangentMode=RCTM_User, time=0.5, value=1}
          {tangentMode=RCTM_User, time=1, value=0}
         ]}}})
    )
   )
   &CameraOffset(Camera Offset Amount=1)
  ]
  rendererClasses [&NiagaraMeshRendererProperties, &NiagaraLightRendererProperties]
  renderers [
   {rendererClass=&NiagaraMeshRendererProperties, rendererIndex=0}
   {rendererClass=&NiagaraLightRendererProperties, rendererIndex=1}
  ]
 }
 FlowersOnGuitar {
  emitterUpdateScript [
   &EmitterState(
    Inactive Response=::NewEnumerator1
    Life Cycle Mode=::NewEnumerator1
    Loop Delay=&RandomRangeFloat(Minimum=0.5, Maximum=2)
    Loop Duration=6
    Recalculate Duration Each Loop=true
    UseLoopDelay=false
   )
   &SpawnRate(Use Spawn Probability=true)
  ]
  particleSpawnScript [
   &PartitionParticles(Number Of Partitions=5, Particles Per Partition=1, Partition Mode:ENiagara_PartitionMode=::NewEnumerator1)
   /SetVariables(Particles.MeshIndex=&RandomRangeInt(Minimum=1, Maximum=5)) {isSetParameters=true, module=SetVariables_#3}
   &InitializeParticle(
    Color=&SelectLinearColorFromArray(
     Array Sampling Mode=::NewEnumerator0
     Direct Array Index=?
     Color Selection Array=DataInterface() {propertyValues {ColorData [
        <rgba 0.0262 0.5748 0.9948 1>
        <rgba 0.2968 0.1864 0.5089 1>
        <rgba 0.6939 0.104 0.6346 1>
        <rgba 0.1479 0.6172 0.6172 1>
        <rgba 1 0.324 0 1>
        <rgba 1 0.9258 0 1>
        <rgba 0.7969 0.0913 0.0913 1>
       ]}}
     Random Seed=0
    )
    Lifetime Max=5
    Lifetime Min=4
    Lifetime Mode=::NewEnumerator1
    Mesh Scale Mode=::NewEnumerator1
    Mesh Uniform Scale Max=0.45
    Mesh Uniform Scale Min=0.374
    Sprite Size Mode=::NewEnumerator1
    UsePositionOffset=false
    Write Material Random=true
   )
   &DynamicMaterialParameters(
    Index 0 Param 1=1
    Index 0 Param 2=0.25
    Index 0 Param 3=0.5
    Module.Index0Param0Write=true
    Module.Index0Param1Write=true
    Module.Index0Param2Write=true
    Module.Index0Param3Write=true
    Module.Index1Param0Write=true
    Module.Index1Param1Write=false
    Module.Index1Param2Write=false
    Module.Index1Param3Write=false
    Write Parameter Index 1=true
   ) {module=DynamicMaterialParameters001}
   &InitialMeshOrientation(
    Mesh Orientation Mode=::NewEnumerator1
    Rotation=&SelectVectorFromArray(
     Direct Array Index=$Particles.PartitionParticles.Partition
     Vector Selection Array=DataInterface() {propertyValues {FloatData [<0 0 0.3>, <0 0 0.3>, <0 0 0.3>, <0 0 0.3>, <0 0 0>, <0 0 0>, <0 0 0>]}}
    )
    Use Rotation Vector=true
   )
  ]
  particleUpdateScript [
   &ParticleState()
   &ScaleMeshSize(Scale Factor=$Engine.Owner.Scale) {enabled=false, module=ScaleMeshSize006}
   &ScaleMeshSize(Scale Factor=&MakeVector(
     X=&FloatFromCurve(FloatCurve=DataInterface() {propertyValues {Curve {keys [
          {arriveTangent=1.8704, interpMode=RCIM_Cubic, leaveTangent=1.8704, tangentMode=RCTM_User, time=0, value=0}
          {arriveTangent=0.2799, leaveTangent=1.1179, tangentMode=RCTM_User, time=0.3, value=1}
          {arriveTangent=0.0224, interpMode=RCIM_Cubic, leaveTangent=0.0224, tangentMode=RCTM_User, time=1, value=1}
         ]}}})
     Y=&FloatFromCurve(FloatCurve=DataInterface() {propertyValues {Curve {keys [
          {arriveTangent=2.4662, interpMode=RCIM_Cubic, leaveTangent=2.4662, tangentMode=RCTM_User, time=0, value=0}
          {arriveTangent=0.422, leaveTangent=1.1179, tangentMode=RCTM_User, time=0.3, value=1}
          {arriveTangent=0.0224, interpMode=RCIM_Cubic, leaveTangent=0.0224, tangentMode=RCTM_User, time=1, value=1}
         ]}}})
     Z=&FloatFromCurve(FloatCurve=DataInterface() {propertyValues {Curve {keys [
          {interpMode=RCIM_Cubic, tangentMode=RCTM_User, time=0, value=0}
          {arriveTangent=0.3068, leaveTangent=0.3068, time=0.2, value=1.025}
          {arriveTangent=0.0743, interpMode=RCIM_Cubic, leaveTangent=0.0743, tangentMode=RCTM_User, time=1, value=1}
         ]}}})
    )) {module=ScaleMeshSize004}
   &ScaleMeshSize(Scale Factor=&MakeVector(
     X=&FloatFromCurve(FloatCurve=DataInterface() {propertyValues {Curve {keys [
          {tangentMode=RCTM_User, time=0, value=1}
          {arriveTangent=-3.4324, interpMode=RCIM_Cubic, leaveTangent=-3.4324, tangentMode=RCTM_User, time=0.7, value=1}
          {arriveTangent=-0.4311, interpMode=RCIM_Cubic, leaveTangent=-0.4311, tangentMode=RCTM_User, time=1, value=0}
         ]}}})
     Y=&FloatFromCurve(FloatCurve=DataInterface() {propertyValues {Curve {keys [
          {tangentMode=RCTM_User, time=0, value=1}
          {arriveTangent=-3.3527, interpMode=RCIM_Cubic, leaveTangent=-3.3527, tangentMode=RCTM_User, time=0.7, value=1}
          {arriveTangent=-0.4311, interpMode=RCIM_Cubic, leaveTangent=-0.4311, tangentMode=RCTM_User, time=1, value=0}
         ]}}})
     Z=&FloatFromCurve(FloatCurve=DataInterface() {propertyValues {Curve {keys [{interpMode=RCIM_Cubic, tangentMode=RCTM_User, time=0.8, value=1}, {interpMode=RCIM_Cubic, tangentMode=RCTM_User, time=1, value=0}]}}})
    )) {module=ScaleMeshSize005}
   &DynamicMaterialParameters(
    Index 0 Param 1=1
    Index 0 Param 2=0.25
    Index 0 Param 3=0.5
    Index 1 Param 1=&FloatFromCurve(
     FloatCurve=DataInterface() {propertyValues {Curve {keys [
         {arriveTangent=-1.6025, interpMode=RCIM_Cubic, leaveTangent=-1.6025, tangentMode=RCTM_User, time=0, value=1}
         {arriveTangent=-0.0559, leaveTangent=-1.4286, tangentMode=RCTM_User, time=0.3, value=0}
         {arriveTangent=0.2127, interpMode=RCIM_Cubic, leaveTangent=0.2127, tangentMode=RCTM_User, time=0.7, value=0}
         {arriveTangent=1.0544, tangentMode=RCTM_User, time=1, value=1}
        ]}}}
     Scale Curve=0.7
    )
    Module.Index0Param0Write=true
    Module.Index0Param1Write=true
    Module.Index0Param2Write=true
    Module.Index0Param3Write=true
    Module.Index1Param0Write=true
    Module.Index1Param1Write=false
    Module.Index1Param2Write=false
    Module.Index1Param3Write=false
    Write Parameter Index 1=true
   )
   &ScaleMeshSize(Scale Factor=&VectorFromFloat(Value=0.75)) {enabled=false, module=ScaleMeshSize003}
   &ScaleMeshSize(Scale Factor=&MakeVector(
     X=&RandomRangeFloat(
      Minimum=&FloatFromCurve(FloatCurve=DataInterface() {propertyValues {Curve {keys [
           {interpMode=RCIM_Cubic, tangentMode=RCTM_User, time=0, value=0.85}
           {interpMode=RCIM_Cubic, tangentMode=RCTM_User, time=0.45, value=1}
           <at 1 0.85 RCIM_Constant>
          ]}}})
      Maximum=&FloatFromCurve(FloatCurve=DataInterface() {propertyValues {Curve {keys [
           {interpMode=RCIM_Cubic, tangentMode=RCTM_User, time=0, value=0.85}
           {interpMode=RCIM_Cubic, tangentMode=RCTM_User, time=0.75, value=1}
           <at 1 0.85 RCIM_Constant>
          ]}}})
     )
     Y=&RandomRangeFloat(
      Minimum=&FloatFromCurve(FloatCurve=DataInterface() {propertyValues {Curve {keys [
           {interpMode=RCIM_Cubic, tangentMode=RCTM_User, time=0, value=0.85}
           {interpMode=RCIM_Cubic, tangentMode=RCTM_User, time=0.45, value=1}
           <at 1 0.85 RCIM_Constant>
          ]}}})
      Maximum=&FloatFromCurve(FloatCurve=DataInterface() {propertyValues {Curve {keys [
           {interpMode=RCIM_Cubic, tangentMode=RCTM_User, time=0, value=0.85}
           {interpMode=RCIM_Cubic, tangentMode=RCTM_User, time=0.75, value=1}
           <at 1 0.85 RCIM_Constant>
          ]}}})
     )
     Z=1
    ))
   &ScaleMeshSize(Scale Factor=&VectorFromFloat(Value=&Lerp_Float(
      A=1
      B=1.4
      Alpha=&FloatFromCurve(
       FloatCurve=DataInterface() {propertyValues {Curve {keys [
           {arriveTangent=9.0725, interpMode=RCIM_Cubic, time=0, value=0}
           <at 0.1 1 RCIM_Cubic>
           {arriveTangent=-0.0088, interpMode=RCIM_Cubic, leaveTangent=-0.0088, tangentMode=RCTM_User, time=1, value=0}
          ]}}}
       CurveIndex=$NPC.NPC_Sparks_Audio_Parameters001.MeasureLifetime
      )
     ))) {module=ScaleMeshSize002}
   &ScaleMeshSize(Scale Factor=&VectorFromFloat(Value=&FloatFromCurve(FloatCurve=DataInterface() {propertyValues {Curve {keys [
          {arriveTangent=0.3651, interpMode=RCIM_Cubic, leaveTangent=0.3651, tangentMode=RCTM_User, time=0, value=0}
          {arriveTangent=-2.5249, leaveTangent=-2.5249, time=0.1408, value=0.9847}
          {arriveTangent=-0.0698, interpMode=RCIM_Cubic, leaveTangent=-0.0698, tangentMode=RCTM_User, time=0.55, value=1}
          {arriveTangent=-0.0405, interpMode=RCIM_Cubic, leaveTangent=-0.0405, tangentMode=RCTM_User, time=1, value=0}
         ]}}}))) {enabled=false, module=ScaleMeshSize001}
   &SkeletalMeshLocation(
    Bone / Socket Index=&SelectIntFromArray(Direct Array Index=$Particles.PartitionParticles.Partition, Int Selection Array=DataInterface() {propertyValues {IntData [1,2,3,4,5,6,7]}})
    Bone / Socket Sampling Mode=::NewEnumerator5
    MaskSpawnGroup=false
    Mesh Sampling Type=::NewEnumerator6
    Sampled Position Offset=&SelectVectorFromArray(
     Direct Array Index=$Particles.PartitionParticles.Partition
     Vector Selection Array=DataInterface() {propertyValues {FloatData [<0 0 0>, <0 0 0>, <0 0 0>, <0 0 0>, <0 0 0>, <0 0 0>, <0 0 0>]}}
    )
    Sampled Position Offset Coordinate Space:ENiagaraSkelMeshTransforms=::NewEnumerator1
    Skeletal Mesh=$System.Skeletal Mesh
   )
   &FN_Light_Attributes(
    Color=$Particles.Color
    Create Color Attribute=true
    Create Exponent Attribute=true
    Create Position Attribute=true
    Create Radius Attribute=true
    Enable Particle Light Enable=true
    Exponent=3
    Light Enabled=true
    Position:NiagaraPosition=$Particles.Position
    Radius=&FloatFromCurve(
     FloatCurve=DataInterface() {propertyValues {Curve {keys [
         {arriveTangent=0.3942, interpMode=RCIM_Cubic, leaveTangent=0.3942, tangentMode=RCTM_User, time=0, value=0}
         {arriveTangent=-2.5249, interpMode=RCIM_Cubic, leaveTangent=-2.5249, tangentMode=RCTM_User, time=0.1408, value=0.9847}
         {arriveTangent=-0.0405, interpMode=RCIM_Cubic, leaveTangent=-0.0405, tangentMode=RCTM_User, time=1, value=0}
        ]}}}
     Scale Curve=12
    )
    Scale Color=&Lerp_Float(
     A=10
     B=20
     Alpha=&FloatFromCurve(FloatCurve=DataInterface() {propertyValues {Curve {keys [
          {interpMode=RCIM_Cubic, tangentMode=RCTM_User, time=0, value=0}
          {arriveTangent=-2.0686, interpMode=RCIM_Cubic, leaveTangent=-2.0686, tangentMode=RCTM_User, time=0.5, value=1}
          {tangentMode=RCTM_User, time=1, value=0}
         ]}}})
    )
   )
   &CameraOffset(Camera Offset Amount=1)
  ]
  rendererClasses [&NiagaraMeshRendererProperties, &NiagaraLightRendererProperties]
  renderers [
   {rendererClass=&NiagaraMeshRendererProperties, rendererIndex=0}
   {rendererClass=&NiagaraLightRendererProperties, rendererIndex=1}
  ]
 }
 FlowersOnGuitar_InitialBurst {
  bEnabled=false
  emitterUpdateScript [
   &EmitterState(
    Inactive Response=::NewEnumerator1
    Life Cycle Mode=::NewEnumerator1
    Loop Behavior=::NewEnumerator1
    Loop Delay=&RandomRangeFloat(Minimum=0.5, Maximum=2)
    Loop Duration=6
    UseLoopDelay=false
   )
   &SpawnBurst_Instantaneous(Spawn Probability=0.65, Use Spawn Probability=false, UseLoopCountLimit=false)
  ]
  particleSpawnScript [
   &PartitionParticles(Number Of Partitions=5, Particles Per Partition=1, Partition Mode:ENiagara_PartitionMode=::NewEnumerator1)
   /SetVariables(Particles.MeshIndex=&RandomRangeInt(Minimum=1, Maximum=5)) {isSetParameters=true, module=SetVariables_#3}
   &InitializeParticle(
    Color=&SelectLinearColorFromArray(
     Array Sampling Mode=::NewEnumerator0
     Direct Array Index=?
     Color Selection Array=DataInterface() {propertyValues {ColorData [
        <rgba 0.0262 0.5748 0.9948 1>
        <rgba 0.2968 0.1864 0.5089 1>
        <rgba 0.6939 0.104 0.6346 1>
        <rgba 0.1479 0.6172 0.6172 1>
        <rgba 1 0.324 0 1>
        <rgba 1 0.9258 0 1>
        <rgba 0.7969 0.0913 0.0913 1>
       ]}}
     Random Seed=0
    )
    Lifetime Max=5
    Lifetime Min=4
    Lifetime Mode=::NewEnumerator1
    Mesh Scale Mode=::NewEnumerator1
    Mesh Uniform Scale Max=0.45
    Mesh Uniform Scale Min=0.374
    Sprite Size Mode=::NewEnumerator1
    UsePositionOffset=false
    Write Material Random=true
   )
   &DynamicMaterialParameters(
    Index 0 Param 1=1
    Index 0 Param 2=0.25
    Index 0 Param 3=0.5
    Module.Index0Param0Write=true
    Module.Index0Param1Write=true
    Module.Index0Param2Write=true
    Module.Index0Param3Write=true
    Module.Index1Param0Write=true
    Module.Index1Param1Write=false
    Module.Index1Param2Write=false
    Module.Index1Param3Write=false
    Write Parameter Index 1=true
   ) {module=DynamicMaterialParameters001}
   &InitialMeshOrientation(
    Mesh Orientation Mode=::NewEnumerator1
    Rotation=&SelectVectorFromArray(
     Direct Array Index=$Particles.PartitionParticles.Partition
     Vector Selection Array=DataInterface() {propertyValues {FloatData [<0 0 0.3>, <0 0 0.3>, <0 0 0.3>, <0 0 0.3>, <0 0 0>, <0 0 0>, <0 0 0>]}}
    )
    Use Rotation Vector=true
   )
  ]
  particleUpdateScript [
   &ParticleState()
   &ScaleMeshSize(Scale Factor=&MakeVector(
     X=&FloatFromCurve(FloatCurve=DataInterface() {propertyValues {Curve {keys [
          {arriveTangent=1.8704, interpMode=RCIM_Cubic, leaveTangent=1.8704, tangentMode=RCTM_User, time=0, value=0}
          {arriveTangent=0.2799, leaveTangent=1.1179, tangentMode=RCTM_User, time=0.3, value=1}
          {arriveTangent=0.0224, interpMode=RCIM_Cubic, leaveTangent=0.0224, tangentMode=RCTM_User, time=1, value=1}
         ]}}})
     Y=&FloatFromCurve(FloatCurve=DataInterface() {propertyValues {Curve {keys [
          {arriveTangent=2.4662, interpMode=RCIM_Cubic, leaveTangent=2.4662, tangentMode=RCTM_User, time=0, value=0}
          {arriveTangent=0.422, leaveTangent=1.1179, tangentMode=RCTM_User, time=0.3, value=1}
          {arriveTangent=0.0224, interpMode=RCIM_Cubic, leaveTangent=0.0224, tangentMode=RCTM_User, time=1, value=1}
         ]}}})
     Z=&FloatFromCurve(FloatCurve=DataInterface() {propertyValues {Curve {keys [
          {interpMode=RCIM_Cubic, tangentMode=RCTM_User, time=0, value=0}
          {arriveTangent=0.3068, leaveTangent=0.3068, time=0.2, value=1.025}
          {arriveTangent=0.0743, interpMode=RCIM_Cubic, leaveTangent=0.0743, tangentMode=RCTM_User, time=1, value=1}
         ]}}})
    )) {module=ScaleMeshSize004}
   &ScaleMeshSize(Scale Factor=&MakeVector(
     X=&FloatFromCurve(FloatCurve=DataInterface() {propertyValues {Curve {keys [
          {tangentMode=RCTM_User, time=0, value=1}
          {arriveTangent=-3.4324, interpMode=RCIM_Cubic, leaveTangent=-3.4324, tangentMode=RCTM_User, time=0.7, value=1}
          {arriveTangent=-0.4311, interpMode=RCIM_Cubic, leaveTangent=-0.4311, tangentMode=RCTM_User, time=1, value=0}
         ]}}})
     Y=&FloatFromCurve(FloatCurve=DataInterface() {propertyValues {Curve {keys [
          {tangentMode=RCTM_User, time=0, value=1}
          {arriveTangent=-3.3527, interpMode=RCIM_Cubic, leaveTangent=-3.3527, tangentMode=RCTM_User, time=0.7, value=1}
          {arriveTangent=-0.4311, interpMode=RCIM_Cubic, leaveTangent=-0.4311, tangentMode=RCTM_User, time=1, value=0}
         ]}}})
     Z=&FloatFromCurve(FloatCurve=DataInterface() {propertyValues {Curve {keys [{interpMode=RCIM_Cubic, tangentMode=RCTM_User, time=0.8, value=1}, {interpMode=RCIM_Cubic, tangentMode=RCTM_User, time=1, value=0}]}}})
    )) {module=ScaleMeshSize005}
   &DynamicMaterialParameters(
    Index 0 Param 1=1
    Index 0 Param 2=0.25
    Index 0 Param 3=0.5
    Index 1 Param 1=&FloatFromCurve(
     FloatCurve=DataInterface() {propertyValues {Curve {keys [
         {arriveTangent=-1.6025, interpMode=RCIM_Cubic, leaveTangent=-1.6025, tangentMode=RCTM_User, time=0, value=1}
         {arriveTangent=-0.0559, leaveTangent=-1.4286, tangentMode=RCTM_User, time=0.3, value=0}
         {arriveTangent=0.2127, interpMode=RCIM_Cubic, leaveTangent=0.2127, tangentMode=RCTM_User, time=0.7, value=0}
         {arriveTangent=1.0544, tangentMode=RCTM_User, time=1, value=1}
        ]}}}
     Scale Curve=0.7
    )
    Module.Index0Param0Write=true
    Module.Index0Param1Write=true
    Module.Index0Param2Write=true
    Module.Index0Param3Write=true
    Module.Index1Param0Write=true
    Module.Index1Param1Write=false
    Module.Index1Param2Write=false
    Module.Index1Param3Write=false
    Write Parameter Index 1=true
   )
   &ScaleMeshSize(Scale Factor=&VectorFromFloat(Value=0.75)) {enabled=false, module=ScaleMeshSize003}
   &ScaleMeshSize(Scale Factor=&MakeVector(
     X=&RandomRangeFloat(
      Minimum=&FloatFromCurve(FloatCurve=DataInterface() {propertyValues {Curve {keys [
           {interpMode=RCIM_Cubic, tangentMode=RCTM_User, time=0, value=0.85}
           {interpMode=RCIM_Cubic, tangentMode=RCTM_User, time=0.45, value=1}
           <at 1 0.85 RCIM_Constant>
          ]}}})
      Maximum=&FloatFromCurve(FloatCurve=DataInterface() {propertyValues {Curve {keys [
           {interpMode=RCIM_Cubic, tangentMode=RCTM_User, time=0, value=0.85}
           {interpMode=RCIM_Cubic, tangentMode=RCTM_User, time=0.75, value=1}
           <at 1 0.85 RCIM_Constant>
          ]}}})
     )
     Y=&RandomRangeFloat(
      Minimum=&FloatFromCurve(FloatCurve=DataInterface() {propertyValues {Curve {keys [
           {interpMode=RCIM_Cubic, tangentMode=RCTM_User, time=0, value=0.85}
           {interpMode=RCIM_Cubic, tangentMode=RCTM_User, time=0.45, value=1}
           <at 1 0.85 RCIM_Constant>
          ]}}})
      Maximum=&FloatFromCurve(FloatCurve=DataInterface() {propertyValues {Curve {keys [
           {interpMode=RCIM_Cubic, tangentMode=RCTM_User, time=0, value=0.85}
           {interpMode=RCIM_Cubic, tangentMode=RCTM_User, time=0.75, value=1}
           <at 1 0.85 RCIM_Constant>
          ]}}})
     )
     Z=1
    ))
   &ScaleMeshSize(Scale Factor=&VectorFromFloat(Value=&Lerp_Float(
      A=1
      B=1.4
      Alpha=&FloatFromCurve(
       FloatCurve=DataInterface() {propertyValues {Curve {keys [
           {arriveTangent=9.0725, interpMode=RCIM_Cubic, time=0, value=0}
           <at 0.1 1 RCIM_Cubic>
           {arriveTangent=-0.0088, interpMode=RCIM_Cubic, leaveTangent=-0.0088, tangentMode=RCTM_User, time=1, value=0}
          ]}}}
       CurveIndex=$NPC.NPC_Sparks_Audio_Parameters001.MeasureLifetime
      )
     ))) {module=ScaleMeshSize002}
   &ScaleMeshSize(Scale Factor=&VectorFromFloat(Value=&FloatFromCurve(FloatCurve=DataInterface() {propertyValues {Curve {keys [
          {arriveTangent=0.3651, interpMode=RCIM_Cubic, leaveTangent=0.3651, tangentMode=RCTM_User, time=0, value=0}
          {arriveTangent=-2.5249, leaveTangent=-2.5249, time=0.1408, value=0.9847}
          {arriveTangent=-0.0698, interpMode=RCIM_Cubic, leaveTangent=-0.0698, tangentMode=RCTM_User, time=0.55, value=1}
          {arriveTangent=-0.0405, interpMode=RCIM_Cubic, leaveTangent=-0.0405, tangentMode=RCTM_User, time=1, value=0}
         ]}}}))) {enabled=false, module=ScaleMeshSize001}
   &SkeletalMeshLocation(
    Bone / Socket Index=&SelectIntFromArray(Direct Array Index=$Particles.PartitionParticles.Partition, Int Selection Array=DataInterface() {propertyValues {IntData [1,2,3,4,5,6,7]}})
    Bone / Socket Sampling Mode=::NewEnumerator5
    MaskSpawnGroup=false
    Mesh Sampling Type=::NewEnumerator6
    Sampled Position Offset=&SelectVectorFromArray(
     Direct Array Index=$Particles.PartitionParticles.Partition
     Vector Selection Array=DataInterface() {propertyValues {FloatData [<0 0 0>, <0 0 0>, <0 0 0>, <0 0 0>, <0 0 0>, <0 0 0>, <0 0 0>]}}
    )
    Sampled Position Offset Coordinate Space:ENiagaraSkelMeshTransforms=::NewEnumerator1
    Skeletal Mesh=$System.Skeletal Mesh
   )
   &FN_Light_Attributes(
    Color=$Particles.Color
    Create Color Attribute=true
    Create Exponent Attribute=true
    Create Position Attribute=true
    Create Radius Attribute=true
    Enable Particle Light Enable=true
    Exponent=3
    Light Enabled=true
    Position:NiagaraPosition=$Particles.Position
    Radius=&FloatFromCurve(
     FloatCurve=DataInterface() {propertyValues {Curve {keys [
         {arriveTangent=0.3942, interpMode=RCIM_Cubic, leaveTangent=0.3942, tangentMode=RCTM_User, time=0, value=0}
         {arriveTangent=-2.5249, interpMode=RCIM_Cubic, leaveTangent=-2.5249, tangentMode=RCTM_User, time=0.1408, value=0.9847}
         {arriveTangent=-0.0405, interpMode=RCIM_Cubic, leaveTangent=-0.0405, tangentMode=RCTM_User, time=1, value=0}
        ]}}}
     Scale Curve=12
    )
    Scale Color=&Lerp_Float(
     A=10
     B=20
     Alpha=&FloatFromCurve(FloatCurve=DataInterface() {propertyValues {Curve {keys [
          {interpMode=RCIM_Cubic, tangentMode=RCTM_User, time=0, value=0}
          {arriveTangent=-2.0686, interpMode=RCIM_Cubic, leaveTangent=-2.0686, tangentMode=RCTM_User, time=0.5, value=1}
          {tangentMode=RCTM_User, time=1, value=0}
         ]}}})
    )
   )
   &CameraOffset(Camera Offset Amount=1)
  ]
  rendererClasses [&NiagaraMeshRendererProperties, &NiagaraLightRendererProperties]
  renderers [
   {rendererClass=&NiagaraMeshRendererProperties, rendererIndex=0}
   {rendererClass=&NiagaraLightRendererProperties, rendererIndex=1}
  ]
 }
 Leaf_local {
  bEnabled=false
  emitterUpdateScript [
   &EmitterState(
    Inactive Response=::NewEnumerator1
    Life Cycle Mode=::NewEnumerator1
    Loop Delay=&RandomRangeFloat(Minimum=0, Maximum=0.5)
    Loop Duration=10
    Recalculate Duration Each Loop=true
    UseLoopDelay=false
   )
   &SpawnRate(Spawn Probability=0.9, Use Spawn Probability=true)
  ]
  particleSpawnScript [
   &PartitionParticles(Number Of Partitions=1) {enabled=false}
   &InitializeParticle(
    Color Maximum=<rgba 0.0499 0.2708 0 1>
    Color Minimum=<rgba 0.0243 0.224 0.009 1>
    Color Mode=::NewEnumerator2
    Lifetime Max=5
    Lifetime Min=4
    Lifetime Mode=::NewEnumerator1
    Mesh Scale Max=<0.175 0.175 0.175>
    Mesh Scale Min=<0.1 0.1 0.1>
    Mesh Scale Mode=::NewEnumerator2
    UsePositionOffset=false
    Write Material Random=true
   )
   &ShapeLocation(
    Box Size=<25 40 1>
    MaskSpawnGroup=false
    Override Seed=false
    Rotation Angle=90
    Rotation Axis=<0 1 0>
    Rotation Mode=::NewEnumerator1
    Shape Primitive=::NewEnumerator2
    Surface Only=false
    UseOwnerScale=true
   )
   &InitialMeshOrientation(Use Rotation Vector=false) {module=InitialMeshOrientation001}
   &DynamicMaterialParameters(
    Index 0 Param 1=0.3
    Index 0 Param 2=0.05
    Index 0 Param 3=1
    Module.Index0Param0Write=true
    Module.Index0Param1Write=true
    Module.Index0Param2Write=true
    Module.Index0Param3Write=true
   )
   &AddVelocity(
    Constrain To Radius=25
    Module.Use Falloff Distance=true
    Origin Offset=<0 -5 0>
    Override Seed=false
    Velocity Mode=::NewEnumerator1
    Velocity Speed=&RandomRangeFloat(Minimum=3, Maximum=6)
   )
  ]
  particleUpdateScript [
   &ParticleState()
   &ScaleMeshSize(Scale Factor=&VectorFromFloat(Value=0.85)) {module=ScaleMeshSize003}
   &SkeletalMeshLocation(
    Bone / Socket Sampling Mode=::NewEnumerator5
    MaskSpawnGroup=false
    Mesh Sampling Type=::NewEnumerator6
    Position Sampling=::NewEnumerator1
    Sampled Position Offset=&AddVectorToPosition(
     Position=$Particles.ShapeLocation.ShapePosition
     Vector=&RandomRangeVector(Minimum=<0 1 0>, Maximum=<0 2 0>)
    )
    Sampled Position Offset Coordinate Space:ENiagaraSkelMeshTransforms=::NewEnumerator1
    Skeletal Mesh=$System.Skeletal Mesh
   )
   &CurlNoiseForce(Noise Strength=5, Pan Noise=true, Pan Noise Field=<0 0 1>)
   &ScaleMeshSize(Scale Factor=&VectorFromFloat(Value=&FloatFromCurve(FloatCurve=DataInterface() {propertyValues {Curve {keys [
          {arriveTangent=0.3651, interpMode=RCIM_Cubic, leaveTangent=0.3651, tangentMode=RCTM_User, time=0, value=0}
          {arriveTangent=-2.5249, leaveTangent=-2.5249, time=0.1408, value=0.9847}
          {arriveTangent=-1.146, interpMode=RCIM_Cubic, leaveTangent=-1.146, time=0.6, value=1}
          {arriveTangent=-0.0405, interpMode=RCIM_Cubic, leaveTangent=-0.0405, tangentMode=RCTM_User, time=1, value=0}
         ]}}})))
   &UpdateMeshOrientation(
    Rotation Rate=&RandomRangeFloat(Minimum=-0.2, Maximum=0.2)
    Rotation Vector=&RandomRangeVector(Minimum=<-1 -1 -1>, Maximum=<1 1 1>)
   )
   &GravityForce(Gravity=<0 0 -3>)
   &Drag(Drag=1.15, Use Linear Drag=true, Use Rotational Drag=true)
   &SolveForcesAndVelocity(Clamp Velocity=false, Limit Acceleration=false)
   &CameraOffset(Camera Offset Amount=1)
  ]
  rendererClasses [&NiagaraMeshRendererProperties]
  renderers [{rendererClass=&NiagaraMeshRendererProperties, rendererIndex=0}]
 }
 Leaf_local001 {
  emitterUpdateScript [
   &EmitterState(
    Inactive Response=::NewEnumerator1
    Life Cycle Mode=::NewEnumerator1
    Loop Delay=&RandomRangeFloat(Minimum=0, Maximum=0.5)
    Loop Duration=10
    Recalculate Duration Each Loop=true
    UseLoopDelay=false
   )
   &SpawnRate(Spawn Probability=0.9, Use Spawn Probability=true)
  ]
  particleSpawnScript [
   &PartitionParticles(Number Of Partitions=1) {enabled=false}
   &InitializeParticle(
    Color Maximum=<rgba 0.0499 0.2708 0 1>
    Color Minimum=<rgba 0.0243 0.224 0.009 1>
    Color Mode=::NewEnumerator2
    Lifetime Max=5
    Lifetime Min=4
    Lifetime Mode=::NewEnumerator1
    Mesh Scale Max=<0.175 0.175 0.175>
    Mesh Scale Min=<0.1 0.1 0.1>
    Mesh Scale Mode=::NewEnumerator2
    UsePositionOffset=false
    Write Material Random=true
   )
   &ShapeLocation(
    Box Size=<25 40 1>
    MaskSpawnGroup=false
    Override Seed=false
    Rotation Angle=90
    Rotation Axis=<0 1 0>
    Rotation Mode=::NewEnumerator1
    Shape Primitive=::NewEnumerator2
    Surface Only=false
    UseOwnerScale=true
   )
   &InitialMeshOrientation(Use Rotation Vector=false) {module=InitialMeshOrientation001}
   &DynamicMaterialParameters(
    Index 0 Param 1=0.3
    Index 0 Param 2=0.05
    Index 0 Param 3=1
    Module.Index0Param0Write=true
    Module.Index0Param1Write=true
    Module.Index0Param2Write=true
    Module.Index0Param3Write=true
   )
   &AddVelocity(
    Constrain To Radius=25
    Module.Use Falloff Distance=true
    Origin Offset=<0 -5 0>
    Override Seed=false
    Velocity Mode=::NewEnumerator1
    Velocity Speed=&RandomRangeFloat(Minimum=3, Maximum=6)
   ) {enabled=false}
  ]
  particleUpdateScript [
   &ParticleState()
   &ScaleMeshSize(Scale Factor=$Engine.Owner.Scale) {enabled=false, module=ScaleMeshSize001}
   &ScaleMeshSize(Scale Factor=&VectorFromFloat(Value=0.85)) {module=ScaleMeshSize003}
   &SkeletalMeshLocation(
    Bone / Socket Sampling Mode=::NewEnumerator5
    MaskSpawnGroup=false
    Mesh Sampling Type=::NewEnumerator6
    Orientation Sampling=::NewEnumerator1
    Sampled Position Offset=&AddVectorToPosition(
     Position=$Particles.ShapeLocation.ShapePosition
     Vector=&RandomRangeVector(Minimum=<0 1 0>, Maximum=<0 2 0>)
    )
    Sampled Position Offset Coordinate Space:ENiagaraSkelMeshTransforms=::NewEnumerator1
    Skeletal Mesh=$System.Skeletal Mesh
   )
   &CurlNoiseForce(Noise Strength=5, Pan Noise=true, Pan Noise Field=<0 0 1>)
   &ScaleMeshSize(Scale Factor=&VectorFromFloat(Value=&FloatFromCurve(FloatCurve=DataInterface() {propertyValues {Curve {keys [
          {arriveTangent=0.3651, interpMode=RCIM_Cubic, leaveTangent=0.3651, tangentMode=RCTM_User, time=0, value=0}
          {arriveTangent=-2.5249, leaveTangent=-2.5249, time=0.1408, value=0.9847}
          {arriveTangent=-1.146, interpMode=RCIM_Cubic, leaveTangent=-1.146, time=0.6, value=1}
          {arriveTangent=-0.0405, interpMode=RCIM_Cubic, leaveTangent=-0.0405, tangentMode=RCTM_User, time=1, value=0}
         ]}}})))
   &UpdateMeshOrientation(
    Rotation Rate=&RandomRangeFloat(Minimum=-0.2, Maximum=0.2)
    Rotation Vector=&RandomRangeVector(Minimum=<-1 -1 -1>, Maximum=<1 1 1>)
   )
   &GravityForce(Gravity=<0 0 -3>) {enabled=false}
   /SetVariables(Particles.Position=&AddVectorToPosition(Position=$Particles.Position, Vector=&MakeVector(X=0, Y=0, Z=0.03))) {isSetParameters=true}
   &Drag(Drag=1.15, Use Linear Drag=true, Use Rotational Drag=true) {enabled=false}
   &SolveForcesAndVelocity(Clamp Velocity=false, Limit Acceleration=false)
   &CameraOffset(Camera Offset Amount=1)
  ]
  rendererClasses [&NiagaraMeshRendererProperties]
  renderers [{rendererClass=&NiagaraMeshRendererProperties, rendererIndex=0}]
 }
}
summary {
 systemName=NS_Guitar_RelayStick
 emitters [
  {
   bEnabled=false
   emitterName=Flower_local
   rendererClasses [&NiagaraMeshRendererProperties, &NiagaraLightRendererProperties]
  }
  {bEnabled=false, emitterName=Leaf_local, rendererClasses [&NiagaraMeshRendererProperties]}
  {emitterName=FlowersOnGuitar, rendererClasses [&NiagaraMeshRendererProperties, &NiagaraLightRendererProperties]}
  {
   bEnabled=false
   emitterName=Flower_local_InitialBurst
   rendererClasses [&NiagaraMeshRendererProperties, &NiagaraLightRendererProperties]
  }
  {
   bEnabled=false
   emitterName=FlowersOnGuitar_InitialBurst
   rendererClasses [&NiagaraMeshRendererProperties, &NiagaraLightRendererProperties]
  }
  {
   emitterName=Flower_local_InitialBurst001
   rendererClasses [&NiagaraMeshRendererProperties, &NiagaraLightRendererProperties]
  }
  {emitterName=Flower_local001, rendererClasses [&NiagaraMeshRendererProperties, &NiagaraLightRendererProperties]}
  {emitterName=Leaf_local001, rendererClasses [&NiagaraMeshRendererProperties]}
 ]
 userVariables []
}
